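; Child weapon Blueprint for the PaddleGinger prop-possession gun that drives its HUD hints and publishes prop selection state.
; BeginPlay maps data table rows to categories; preview index changes update category and cycle text, assuming equal-sized categories.
; Vehicle, visibility and interaction changes, with a timeout for stuck interactions, refresh key-hint visibility.
; PropGunFunctionLibrary receives prop index and an activated state that also counts being hidden, driving or interacting.
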(event EventBeginPlay
  (|CallParentFunction)
  (Utilities|FlowControl|Sequence
    (:then_0
      ;[1 1 1] Create category lookup table
      (Utilities|Array|Clear (Variables|Default|GetCategoryPropIndexMapping))
      (bind _outrownames (DataTable|GetDataTableRowNames (Variables|Default|GetPropDataTable)))
      (for _array_element _outrownames
        (bind _returnvalue (Utilities|GetDataTableRow (Variables|Default|GetPropDataTable) _array_element))
        (Utilities|Array|Add (Variables|Default|GetCategoryPropIndexMapping) (.CategoryLabelText (Utilities|Struct|BreakPossessPropClasses _returnvalue)))
        (Utilities|Array|AddUnique (Variables|Default|GetCategories) (.CategoryLabelText (Utilities|Struct|BreakPossessPropClasses _returnvalue)))))
    (:then_1
      ;[1 1 1] Initialize Category and Prop Text
      (bind _outcategorytext (CallFunction|SetCategoryText 0))
      (bind _outpropindex (CallFunction|SetPropText 0 (Variables|Default|GetNumItemsPerCategory)))
      (CallFunction|BroadcastItemIndexChanged _outcategorytext _outpropindex))
    (:then_2
      ;[1 1 1] Watch for Pawn Enter/Exit vehicle
      (bind _asplayer_pawn_athena (Utilities|Casting|CastToPlayerPawn_Athena (Variables|Default|GetInstigatorFortPawn)))
      (Default|BindEventtoOnEnteredVehicleDispatcher _asplayer_pawn_athena (delegate Custom|Pawn%20Entered%20Vehicle))
      (Default|BindEventtoOnExitedVehicleDispatcher _asplayer_pawn_athena (delegate Custom|PawnExitedVehicle))
      (EventDispatchers|BindEventtoOnStartedInteractSearch _asplayer_pawn_athena (delegate Custom|CustomEvent_0))
      (EventDispatchers|BindEventtoOnEndedInteractSearch _asplayer_pawn_athena (delegate Custom|CustomEvent_3))
      (EventDispatchers|BindEventtoOnInteractionFailedCallback _asplayer_pawn_athena (delegate Custom|CustomEvent_4))
      (EventDispatchers|BindEventtoOnAttemptedInstantInteractSearch _asplayer_pawn_athena (delegate Custom|CustomEvent_2)))))

(event EventTick
  (params (DeltaSeconds :type Float%20%28single-precision%29)))

(event Custom|Pawn%20Entered%20Vehicle
  ;[1 1 1] Watch for Pawn Enter/Exit vehicle
  (CallFunction|UpdateHUDFromVehicleState))

(event Custom|PawnExitedVehicle
  ;[1 1 1] Watch for Pawn Enter/Exit vehicle
  (CallFunction|UpdateHUDFromVehicleState))

(event Custom|CustomEvent_2
  (Development|PrintString "Attempted Instant Interact Search" true true "(R=0.000000,G=0.660000,B=1.000000,A=1.000000)" 2e+01))

(event Custom|CustomEvent_0
  (CallFunction|SetInteracting true))

(event Custom|CustomEvent_4
  (params (FortPlayerPawnAthena :type Fort%20Player%20Pawn%20Athena%20Object%20Reference))
  (CallFunction|SetInteracting false))

(event Custom|CustomEvent_3
  (CallFunction|SetInteracting false))

(event EventOnWeaponVisibilityChanged
  (params (bVisible :type Boolean) (bSetForLocalControllerOnly :type Boolean))
  (if (Math|Boolean|NotEqual(Boolean) (Variables|Default|GetbIsVisible) bVisible)
    (Variables|Default|SetbIsVisible bVisible)
    (CallFunction|SetContextHelpVisiblity (Variables|Default|GetCached%20Prop%20Activated) (Variables|Default|GetCached%20Prop%20Placed))))

(event Custom|SetInteracting
  (params (InIsInteracting :type Boolean))
  (bind _output_get (Variables|Default|SetbPawnIsInteracting InIsInteracting))
  (if _output_get
    (bind _returnvalue (Utilities|Time|SetTimerbyEvent (delegate Custom|Interaction%20didn't%20end) 1e+01))
    (Variables|Default|SetInteractionTimer _returnvalue)
    (CallFunction|SetContextHelpVisiblity (Variables|Default|GetCached%20Prop%20Activated) (Variables|Default|GetCached%20Prop%20Placed))
    (else
      (Utilities|Time|ClearandInvalidateTimerbyHandle (Variables|Default|GetInteractionTimer))
      (CallFunction|SetContextHelpVisiblity (Variables|Default|GetCached%20Prop%20Activated) (Variables|Default|GetCached%20Prop%20Placed)))))

(event Custom|Interaction%20didn't%20end
  (CallFunction|SetInteracting false))

(event FortWeapon|EventOnEquippedAndAttached
  (|CallParentFunction)
  (CallFunction|UpdateHUDFromVehicleState)
  (Variables|Default|SetbIsVisible (Rendering|IsVisible (Variables|Visual|GetWeaponMesh)))
  (Variables|Default|SetbWeaponIsEquipped true)
  (Variables|Default|SetbPawnIsInteracting false)
  (CallFunction|SetContextHelpVisiblity (Variables|Default|GetCached%20Prop%20Activated) (Variables|Default|GetCached%20Prop%20Placed)))

(event FortWeapon|EventOnLocalUnEquip
  (|CallParentFunction)
  (Variables|Default|SetbWeaponIsEquipped false))

(event Custom|UpdateHUDFromVehicleState
  (Utilities|IsinVehicle (Variables|Default|GetInstigatorFortPawn)
    (:true
      (if (not (Variables|Default|GetbPawnInVehicle))
        (Variables|Default|SetbPawnInVehicle true)))
    (:false
      (if (Variables|Default|GetbPawnInVehicle)
        (Variables|Default|SetbPawnInVehicle false))))
  (CallFunction|SetContextHelpVisiblity (Variables|Default|GetCached%20Prop%20Activated) (Variables|Default|GetCached%20Prop%20Placed)))

(fn BroadcastItemIndexChanged
  (params (CategoryLabelText :type Text) (PropIndex :type Integer))
  (bind _asfort_player_pawn_athena (Utilities|Casting|CastToFortPlayerPawnAthena (Variables|Default|GetInstigatorFortPawn)))
  (Class|PropGunFunctionLibrary|UpdatePropIndexEventState _asfort_player_pawn_athena (Utilities|Struct|MakePaddleGingerPropData CategoryLabelText PropIndex (Variables|Default|GetNumItemsPerCategory) (Utilities|Array|FindItem (Variables|Default|GetCategories) CategoryLabelText) (Utilities|Array|Length (Variables|Default|GetCategories)) (Variables|/Script/Engine.Pawn|GetPlayerState _asfort_player_pawn_athena))))

(fn OnSetCuratedPreviewItemIndex
  (|CallParentFunction)
  (Utilities|FlowControl|Sequence
    (:then_0
      (bind _outcategorytext (CallFunction|SetCategoryText (Variables|Default|GetCuratedItemPreviewIndex))))
    (:then_1
      (bind _outpropindex (CallFunction|SetPropText (Math|Integer|%25%28Integer%29 (Variables|Default|GetCuratedItemPreviewIndex) (Variables|Default|GetNumItemsPerCategory)) (Variables|Default|GetNumItemsPerCategory)))
      (CallFunction|BroadcastItemIndexChanged _outcategorytext _outpropindex))))

(fn SetContextHelpVisiblity
  (params (PropActivated :type Boolean) (PropPlaced :type Boolean))
  (bind _propactivated (Variables|GetPropActivated))
  (Variables|Default|SetCached%20Prop%20Activated PropActivated)
  (Variables|Default|SetCached%20Prop%20Placed PropPlaced)
  (|CallParentFunction PropActivated PropPlaced)
  (if (Variables|Default|GetbWeaponIsEquipped)
    (WeaponHUDData|SetHudKeyActionVisibilitybyID "Deactivate" _propactivated)
    (WeaponHUDData|SetHudKeyActionVisibilitybyID "SettleInPlace" (and _propactivated (not (Variables|GetPropPlaced))))
    (WeaponHUDData|SetHudKeyActionVisibilitybyID "EndSettle" (Variables|GetPropPlaced))
    (bind _asfort_player_pawn_athena (Utilities|Casting|CastToFortPlayerPawnAthena (Variables|Default|GetInstigatorFortPawn)))
    (Class|PropGunFunctionLibrary|UpdatePropActivatedEventState _asfort_player_pawn_athena (Utilities|Struct|MakePaddleGingerPropActivatedData (Math|Boolean|ORBoolean (Variables|GetPropActivated) (not (Variables|Default|GetbIsVisible)) (Variables|Default|GetbPawnInVehicle) (Variables|Default|GetbPawnIsInteracting)) (Variables|/Script/Engine.Pawn|GetPlayerState _asfort_player_pawn_athena)))))

(fn SetCategoryText
  (params (InPropIndex :type Integer))
  (bind _result (ArrayMacros|SafeArrayGet (Variables|Default|GetCategoryPropIndexMapping) InPropIndex)
    (:%20%20
      (WeaponHUDData|SetHudKeyActionDescriptionbyID "Category" (Utilities|Text|FormatText "NSLOCTEXT(\"[44EA504742B88D4654CA53BB03691FEA]\", \"2E558CAD456CA5305EFC8382FEC4290B\", \"Category: {Category}\")" _result))
      (return _result))))

(fn SetPropText
  (params (PropCategoryIndex :type Integer) (CategoryPropCount :type Integer))
  (WeaponHUDData|SetHudKeyActionDescriptionbyID "CycleProp" (Utilities|Text|FormatText "NSLOCTEXT(\"[44EA504742B88D4654CA53BB03691FEA]\", \"710EC6B44FB6BDD5F19AA09F40A39605\", \"Cycle prop ({0} of {1})\")" (+ PropCategoryIndex 1) CategoryPropCount))
  (return PropCategoryIndex))

(fn ConstructionScript
  (|CallParentFunction))
