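; Fortnite weapon Blueprint for PaddleGinger that shows a floating preview of a prop class beside the wielder, linked by a Niagara beam.
; Equip or PreviewClass replication extracts the class's first static mesh, scales it to fit MaxSize and applies fade-in dynamic materials copying its textures.
; Tick moves the preview toward a socket-based target via native helper, spring or clamped lerp, snapping after teleports.
; Switch_FX pulses the beam and shows a sound indicator.
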(event EventTick
  (params (DeltaSeconds :type Float%20%28single-precision%29))
  (bind _staticmeshpreview (Variables|Default|GetStaticMeshPreview))
  (bind _returnvalue (Transformation|GetWorldLocation _staticmeshpreview))
  (bind _staticmeshpreview_2 (Variables|Default|GetStaticMeshPreview))
  (bind _minzoffsetfromtarget (Variables|Default|GetMinZOffsetFromTarget))
  (if (ScalableFloat|GetValueasBool (Variables|Default|GetShouldUseNative))
    ;[1 1 1] Put the preview prop in place.  I wish we had a flexible component to do this kind of stuff.
    (Class|PropGunFunctionLibrary|UpdatePropPreviewPosition (Variables|Default|GetStaticMeshPreview) DeltaSeconds (Variables|Default|GetInstigatorFortPawn) (Variables|Default|GetAttach%20Socket%20Name) (Variables|Visual|GetWeaponMesh) (Variables|Default|GetRelativeTranslationWhenAttached) (Variables|Default|GetMinZOffsetFromTarget) (Variables|Default|GetRelativeRotationWhenAttached) (Variables|Default|GetCachedWorldScale) 2e+01)
    (else
      (bind _newparam (CallFunction|GetFrameTransformTarget))
      (bind (_newparam_location _newparam_rotation _newparam_scale) (Math|Transform|BreakTransform _newparam))
      (if (> (Math|Vector|DistanceSquared(Vector) _returnvalue _newparam_location) 1e+06)
        ;;[1 1 1] If we've moved super far (i.e. teleported) don't spring/lerp
        (Transformation|SetWorldTransform (Variables|Default|GetStaticMeshPreview) (Math|Transform|MakeTransform _newparam_location _newparam_rotation _newparam_scale) false false)
        (elif (Variables|Default|GetbUseSpring)
          (bind _returnvalue_2 (Math|Interpolation|VectorSpringInterp _returnvalue _newparam_location (Variables|Default|GetSpring%20State) 7e+01 0.5 DeltaSeconds))
          (Transformation|SetWorldTransform (Variables|Default|GetStaticMeshPreview) (Math|Transform|MakeTransform _returnvalue_2 _newparam_rotation _newparam_scale) false false)
          (else
            (if (< (- (+ (.z _newparam_location) _minzoffsetfromtarget) (.z _returnvalue)) 0.0)
              (Transformation|SetWorldTransform _staticmeshpreview_2 (Math|Transform|MakeTransform (Math|Interpolation|VInterpTo _returnvalue _newparam_location DeltaSeconds 2e+01) _newparam_rotation _newparam_scale) false false)
              (else
                ;;[1 1 1] Don't allow it to go below the minZOffset to prevent th ebeam from turning upsidedown
                (Transformation|SetWorldTransform _staticmeshpreview_2 (Math|Transform|MakeTransform (Math|Vector|MakeVector (.x (Math|Interpolation|VInterpTo _returnvalue _newparam_location DeltaSeconds 2e+01)) (.y (Math|Interpolation|VInterpTo _returnvalue _newparam_location DeltaSeconds 2e+01)) (+ (.z _newparam_location) _minzoffsetfromtarget)) _newparam_rotation _newparam_scale) false false)))))))))

(event Custom|StartFrameUpdateForPropPreview
  ;[1 1 1] Put the preview prop in place.  I wish we had a flexible component to do this kind of stuff.
  (Actor|Tick|SetActorTickEnabled true)
  (bind _newparam (CallFunction|GetFrameTransformTarget))
  (Transformation|SetWorldTransform (Variables|Default|GetStaticMeshPreview) _newparam false false))

(event Custom|SetupAndUpdateBeam
  (bind _staticmeshpreview (Variables|Default|GetStaticMeshPreview))
  (bind (_min _max) (Components|StaticMesh|GetLocalBounds _staticmeshpreview))
  (bind _x (.x _min))
  (bind _x_2 (.x _max))
  (bind _returnvalue (- _x_2 _x))
  (bind _returnvalue_2 (Transformation|GetWorldScale _staticmeshpreview))
  (bind _x_3 (.x _returnvalue_2))
  (bind _returnvalue_3 (* _x_3 _returnvalue))
  (Utilities|FlowControl|Sequence
    (:then_0
      (Components|Activation|Activate (Variables|Default|GetNS_PaddleGinger_Gun_Ray) false)
      (Components|Activation|Activate (Variables|Default|GetNS_PaddleGinger_Target_Glow) false))
    (:then_1
      (Networking|SwitchHasAuthority
        (:Remote
          (Niagara|SetNiagaraVariable(Object) (Variables|Default|GetNS_PaddleGinger_Gun_Ray) "User.Prop_Component" (Variables|Default|GetStaticMeshPreview))
          (Niagara|SetNiagaraVariable(Float) (Variables|Default|GetNS_PaddleGinger_Gun_Ray) "Scale_Size" _returnvalue_3)
          (Niagara|SetNiagaraVariable(Float) (Variables|Default|GetNS_PaddleGinger_Target_Glow) "Scale_Size" _returnvalue_3))))
    (:then_2
      (Transformation|AttachComponentToComponent (Variables|Default|GetNS_PaddleGinger_Target_Glow) (Variables|Default|GetStaticMeshPreview) "root" "KeepRelative" "KeepRelative" "KeepRelative" true))))

(event EventBeginPlay
  (bind _returnvalue (Game|GetInstigator))
  (Utilities|IsValid _returnvalue
    (:Is%20Valid
      (bind _asfort_pawn (Utilities|Casting|CastToFortPawn _returnvalue))
      (Variables|Default|SetInstigatorFortPawn _asfort_pawn))))

(event FortWeapon|EventOnLocalUnEquip
  (Components|StaticMesh|SetStaticMesh (Variables|Default|GetStaticMeshPreview) 0)
  (Actor|Tick|SetActorTickEnabled false)
  (Niagara|SetNiagaraVariable(Object) (Variables|Default|GetNS_PaddleGinger_Gun_Ray) "User.Prop_Component" 0))

(event Custom|Switch_FX
  (SoundIndicator|ShowSoundIndicator (Variables|SoundIndicator|GetSoundIndicatorComponent) (Transformation|GetActorLocation) "/MotherGameplay/Sounds/PaddleGinger/PaddleGinger_Select_01.PaddleGinger_Select_01" 1.0 0)
  (bind (_direction _pulse_fx _pulse_brightness) (Timeline|Timeline_0 PlayFromStart)
    (:Update
      (Niagara|SetNiagaraVariable(Float) (Variables|Default|GetNS_PaddleGinger_Gun_Ray) "Pulse_FX" _pulse_fx)
      (Niagara|SetNiagaraVariable(Float) (Variables|Default|GetNS_PaddleGinger_Gun_Ray) "Pulse_Brightness" _pulse_brightness))))

(event FortWeapon|EventOnEquippedAndAttached
  (CallFunction|SetupWeaponPreviewClass)
  (CallFunction|SetContextHelpVisiblity false)
  (CallFunction|OnSetCuratedPreviewItemIndex)
  (CallFunction|PlayFadeinanimation)
  ;[1 1 1] Weapon disables the tick in AFortWeapon::SetAllowWeaponTicking we need to re-enable it
  (Utilities|FlowControl|DelayUntilNextTick)
  (Actor|Tick|SetActorTickEnabled true))

(event Custom|PlayFadeinanimation
  (bind (_direction _newtrack_0) (Timeline|FadeinPreviewItem Play)
    (:Update
      (for _array_element (Variables|Default|GetMID_StaticMeshPreview)
        (Rendering|Material|SetScalarParameterValue _array_element "FadeIn" _newtrack_0 :!class /Script/Engine.MaterialInstanceDynamic)))))

(fn GetFrameTransformTarget
  (bind (_returnvalue_location _returnvalue_rotation) (Math|Transform|BreakTransform (Transformation|GetSocketTransform (Variables|/Script/Engine.Character|GetMesh (Variables|Default|GetInstigatorFortPawn)) (Variables|Default|GetAttach%20Socket%20Name) "RTS_World")))
  (bind (_returnvalue_location_x _returnvalue_location_y _returnvalue_location_z) (Math|Vector|BreakVector _returnvalue_location))
  (bind (_x _y) (Math|Rotator|GetAxes _returnvalue_rotation))
  (bind _returnvalue_location_2 (Math|Transform|BreakTransform (Transformation|GetActorTransform (Variables|Default|GetInstigatorFortPawn))))
  (bind (_returnvalue_location_x_2 _returnvalue_location_y_2) (Math|Vector|BreakVector _returnvalue_location_2))
  (bind (_returnvalue_location_3 _returnvalue_rotation_2 _returnvalue_scale) (Math|Transform|BreakTransform (Transformation|GetWorldTransform (Variables|Visual|GetWeaponMesh))))
  (bind (_x_2 _y_2 _z) (Math|Rotator|BreakRotIntoAxes (Math|Conversions|RotationFromXVector (Math|Vector|Normalize (* _y "1.000000,1.000000, 0")))))
  (bind (_relativetranslationwhenattached_x _relativetranslationwhenattached_y _relativetranslationwhenattached_z) (Math|Vector|BreakVector (Variables|GetRelativeTranslationWhenAttached_X)))
  (return (Math|Vector|vector+vector (Math|Vector|MakeVector _returnvalue_location_x_2 _returnvalue_location_y_2 _returnvalue_location_z) (* _x_2 _relativetranslationwhenattached_x) (* _y_2 _relativetranslationwhenattached_y) (* _z _relativetranslationwhenattached_z)) (Variables|Default|GetRelativeRotationWhenAttached) (* _returnvalue_scale (Variables|Default|GetCachedWorldScale))))

(fn GetAffiliationsToShowFor
  (return (Utilities|Array|MakeArray "Hostile")))

(fn GetRelativeTransformForPreview
  (params (ParentActor :type Actor%20Object%20Reference))
  (bind (_returnvalue_location _returnvalue_rotation) (Math|Transform|BreakTransform (Transformation|GetActorTransform ParentActor)))
  (bind (_returnvalue_rotation_roll _returnvalue_rotation_pitch _returnvalue_rotation_yaw) (Math|Rotator|BreakRotator _returnvalue_rotation))
  (bind (_x _y _z) (Math|Rotator|GetAxes (Math|Rotator|MakeRotator _returnvalue_rotation_roll _returnvalue_rotation_pitch _returnvalue_rotation_yaw)))
  (return (Math|Transform|MakeTransform (Math|Vector|vector+vector _returnvalue_location (* _x 4e+01) (* _y 3e+01) (* _z 9e+01)) (Math|Rotator|MakeRotator _returnvalue_rotation_roll _returnvalue_rotation_pitch (+ _returnvalue_rotation_yaw 45.0)) (* (Transformation|GetActorScale3D ParentActor) (Variables|Default|GetRelativeScale)))))

(fn GetScaleForExtents
  (params (Extents :type Vector))
  (return (Math|Conversions|ToVector(Float) (Math|Float|Min(Float) (/ (.x (Variables|Default|GetMaxSize)) (.x Extents)) (/ (.y (Variables|Default|GetMaxSize)) (.y Extents)) (/ (.z (Variables|Default|GetMaxSize)) (.z Extents))))))

(fn OnRep_CuratedItemIndex)

(fn OnRep_CuratedItemPreviewIndex
  (CallFunction|OnSetCuratedPreviewItemIndex))

(fn OnRep_PreviewClass
  (CallFunction|SetupWeaponPreviewClass)
  (CallFunction|PlayFadeinanimation))

(fn OnRep_WeaponPreviewActor
  (bind _weaponpreviewactor (Variables|FortWeapon|GetWeaponPreviewActor))
  (bind _instigatorfortpawn (Variables|Default|GetInstigatorFortPawn))
  (CallFunction|OnSetWeaponPreviewActor)
  (Utilities|IsValid (Variables|FortWeapon|GetWeaponPreviewActor)
    (:Is%20Valid
      (Collision|SetActorEnableCollision (Variables|FortWeapon|GetWeaponPreviewActor) false)
      (Development|PrintString "Weapon Preview Actor Set On Weapon")
      (Utilities|IsValid _instigatorfortpawn
        (:Is%20Valid
          (bind _asfort_player_pawn (Utilities|Casting|CastToFortPlayerPawn _instigatorfortpawn))
          (Transformation|SetMobility (Variables|/Script/Engine.Actor|GetRootComponent _weaponpreviewactor) "Movable")
          (Transformation|SetActorTransform _weaponpreviewactor (Math|Transform|MakeTransform (.location (CallFunction|GetRelativeTransformforPreview _asfort_player_pawn)) (.rotation (CallFunction|GetRelativeTransformforPreview _asfort_player_pawn)) (Transformation|GetActorScale3D _weaponpreviewactor)) false false)
          (Transformation|AttachActorToComponent _weaponpreviewactor (Variables|/Script/Engine.Character|GetMesh _asfort_player_pawn) "head" "KeepWorld" "KeepWorld" "KeepWorld" true)
          (Collision|SetCollisionEnabled (Variables|/Script/Engine.Character|GetMesh _asfort_player_pawn) "NoCollision"))))
    (:Is%20Not%20Valid
      (Development|PrintString "Weapon preview actor set to invalid" true true "(R=0.000000,G=0.660000,B=1.000000,A=1.000000)" 1e+01))))

(fn OnSetCuratedPreviewItemIndex)

(fn OnSetPreviewClass)

(fn OnSetWeaponPreviewActor)

(fn SetContextHelpVisiblity
  (params (PropActivated :type Boolean) (PropPlaced :type Boolean)))

(fn SetMaterialsonPreviewMesh
  (bind _staticmeshpreview (Variables|Default|GetStaticMeshPreview))
  (bind _staticmesh (Variables|/Script/Engine.StaticMeshComponent|GetStaticMesh _staticmeshpreview))
  (bind _staticmeshpreview_2 (Variables|Default|GetStaticMeshPreview))
  (Utilities|Array|Clear (Variables|Default|GetMID_StaticMeshPreview))
  (for _index (range 2)
    (Utilities|IsValid (StaticMesh|GetMaterial _staticmesh _index)
      (:Is%20Valid
        (bind _returnvalue (Rendering|Material|CreateDynamicMaterialInstance (StaticMesh|GetMaterial _staticmesh _index) "None" "None" :!class /Script/Engine.KismetMaterialLibrary))
        (bind _returnvalue_2 (Rendering|Material|GetTextureParameterValue _returnvalue "Diffuse" :!class /Script/Engine.MaterialInstanceDynamic))
        (bind _returnvalue_3 (Rendering|Material|GetTextureParameterValue _returnvalue "Normals" :!class /Script/Engine.MaterialInstanceDynamic))
        (bind _returnvalue_4 (Rendering|Material|GetTextureParameterValue _returnvalue "MaskTexture" :!class /Script/Engine.MaterialInstanceDynamic))
        (bind _returnvalue_5 (Rendering|Material|CreateDynamicMaterialInstance _staticmeshpreview_2 _index "/MotherGameplay/Items/PaddleGinger/Assets/Materials/M_PaddleGingerProp.M_PaddleGingerProp" "None" :!class /Script/Engine.PrimitiveComponent))
        (Rendering|Material|SetTextureParameterValue _returnvalue_5 "Diffuse" _returnvalue_2)
        (Rendering|Material|SetTextureParameterValue _returnvalue_5 "Normals" _returnvalue_3)
        (Utilities|IsValid _returnvalue_4
          (:Is%20Valid
            (Rendering|Material|SetTextureParameterValue _returnvalue_5 "MaskTexture" _returnvalue_4))
          (:Is%20Not%20Valid
            (Rendering|Material|SetTextureParameterValue _returnvalue_5 "MaskTexture" "/Game/Characters/Player/Female/Medium/Bodies/F_MED_Commando_01/Textures/SolidWhite_MaskCompression.SolidWhite_MaskCompression")))
        (Rendering|Material|SetScalarParameterValue _returnvalue_5 "FadeIn" 0.0 :!class /Script/Engine.MaterialInstanceDynamic)
        (Rendering|Material|SetMaterial _staticmeshpreview_2 _index _returnvalue_5)
        (Utilities|Array|Add (Variables|Default|GetMID_StaticMeshPreview) _returnvalue_5)))))

(fn SetupWeaponPreviewClass
  (bind _preview_class (Variables|Default|GetPreview%20Class))
  (bind _outstaticmeshes (Fortnite|GetStaticMeshesforClass _preview_class))
  (bind _returnvalue (Transformation|GetWorldLocation (Variables|Default|GetStaticMeshPreview)))
  (bind _returnvalue_2 (Transformation|GetWorldRotation (Variables|Default|GetStaticMeshPreview)))
  (bind (_returnvalue_location _returnvalue_rotation _returnvalue_scale) (Math|Transform|BreakTransform (Transformation|GetRelativeTransform (Variables|Default|GetStaticMeshPreview))))
  (bind (_x _y _z) (Math|Rotator|GetAxes _returnvalue_2))
  (bind _staticmeshpreview (Variables|Default|GetStaticMeshPreview))
  (bind (_min _max) (Components|StaticMesh|GetLocalBounds _staticmeshpreview))
  (bind _returnvalue_3 (- _max _min))
  (bind _relativescale (Variables|Default|GetRelativeScale))
  (bind _relativescale3d (Variables|/Script/Engine.SceneComponent|GetRelativeScale3D (Variables|Default|GetStaticMeshPreview)))
  (bind _returnvalue_4 (* _relativescale3d _relativescale))
  (bind _staticmeshpreview_2 (Variables|Default|GetStaticMeshPreview))
  (bind _attach_socket_name (Variables|Default|GetAttach%20Socket%20Name))
  (CallFunction|OnSetPreviewClass)
  (if (Utilities|IsValidClass (Variables|Default|GetPreview%20Class))
    ;[1 1 1] Right now just grab the first one.
    (bind _output_get (LocalVariables|SetAvailable%20Meshes _outstaticmeshes))
    (if (Utilities|Array|IsValidIndex _output_get)
      (Components|StaticMesh|SetStaticMesh (Variables|Default|GetStaticMeshPreview) (Utilities|Array|Get(acopy) _outstaticmeshes))
      (CallFunction|SetMaterialsonPreviewMesh)
      (bind _asfort_player_pawn (Utilities|Casting|CastToFortPlayerPawn (Variables|Default|GetInstigatorFortPawn)))
      (bind _output_get_2 (LocalVariables|SetPreviewTransform (CallFunction|GetRelativeTransformforPreview _asfort_player_pawn)))
      (Transformation|SetWorldTransform _staticmeshpreview_2 _output_get_2 false false)
      (if (Variables|Default|GetDoFrameUpdateForPreviewTransform)
        (Transformation|DetachFromComponent _staticmeshpreview_2 "KeepRelative" "KeepRelative" "KeepRelative" true)
        (elif (Variables|Default|GetAttachToWeapon)
          (LocalVariables|SetComponentForAttach (Variables|Visual|GetWeaponMesh))
          (Transformation|AttachComponentToComponent _staticmeshpreview_2 (LocalVariables|GetComponentForAttach) (Variables|Default|GetAttach%20Socket%20Name) "KeepWorld" "KeepWorld" "KeepWorld" true)
          (elif (Variables|Default|GetAttachToCapsule)
            (LocalVariables|SetComponentForAttach (Variables|/Script/Engine.Character|GetCapsuleComponent _asfort_player_pawn))
            (else
              (LocalVariables|SetComponentForAttach (Variables|/Script/Engine.Character|GetMesh _asfort_player_pawn))))))
      (Collision|SetCollisionEnabled _staticmeshpreview_2 "NoCollision")
      (Development|PrintString "Hello")
      ;[1 1 1] debug attachment
      (Development|PrintString (Utilities|String|Append "Location:" (Utilities|String|ToString(Vector) _returnvalue_location) "   Rotation:" (Utilities|String|ToString(Rotator) _returnvalue_rotation) "  Scale:" (Utilities|String|ToString(Vector) _returnvalue_scale)) (Variables|Default|GetdebugAttachmentTransform) (Variables|Default|GetdebugAttachmentTransform) "(R=1.000000,G=1.000000,B=1.000000,A=1.000000)" 2e+01)
      (if (Variables|Default|GetDoFrameUpdateForPreviewTransform)
        (Transformation|SetRelativeTransform _staticmeshpreview_2 (Math|Transform|MakeTransform "0, 0, 0" "0, 0, 0" _returnvalue_4) false false)
        (else
          (Transformation|SetRelativeTransform _staticmeshpreview_2 (Math|Transform|MakeTransform (Variables|Default|GetRelativeTranslationWhenAttached) (Variables|Default|GetRelativeRotationWhenAttached) _returnvalue_4) false false)))
      (if (Variables|Default|GetdebugAttachmentTransform)
        (Rendering|Debug|DrawDebugLine _returnvalue (+ _returnvalue (* _x 1e+02)) "(R=1.000000,G=0.000000,B=0.000000,A=1.000000)" 2e+01 2.0)
        (Rendering|Debug|DrawDebugLine _returnvalue (+ _returnvalue (* _y 1e+02)) "(R=0.000000,G=1.000000,B=0.303152,A=1.000000)" 2e+01 2.0)
        (Rendering|Debug|DrawDebugLine _returnvalue (+ _returnvalue (* _z 1e+02)) "(R=0.000000,G=0.137348,B=1.000000,A=1.000000)" 2e+01 2.0)
        ;[1 1 1] draw socket transform
        (bind (_returnvalue_location_2 _returnvalue_rotation_2) (Math|Transform|BreakTransform (Transformation|GetSocketTransform (Variables|/Script/Engine.Character|GetMesh _asfort_player_pawn) _attach_socket_name "RTS_World")))
        (Rendering|Debug|DrawDebugLine _returnvalue_location_2 (+ _returnvalue_location_2 (* (Math|Rotator|GetAxes _returnvalue_rotation_2) 1e+02)) "(R=1.000000,G=0.000000,B=0.000000,A=1.000000)" 2e+01 1.0)
        (Rendering|Debug|DrawDebugLine _returnvalue_location_2 (+ _returnvalue_location_2 (* (.Y (Math|Rotator|GetAxes _returnvalue_rotation_2)) 1e+02)) "(R=0.000000,G=1.000000,B=0.303152,A=1.000000)" 2e+01 1.0)
        (Rendering|Debug|DrawDebugLine _returnvalue_location_2 (+ _returnvalue_location_2 (* (.Z (Math|Rotator|GetAxes _returnvalue_rotation_2)) 1e+02)) "(R=0.000000,G=0.137348,B=1.000000,A=1.000000)" 2e+01 1.0)
        (else
          _))
      (Variables|Default|SetCachedExtentsHalf (* _returnvalue_3 0.5))
      (bind _output_get_3 (Variables|Default|SetCachedWorldScale (* (CallFunction|GetScaleforExtents _returnvalue_3) (Variables|Default|GetRelativeScale))))
      (Transformation|SetWorldScale3D _staticmeshpreview _output_get_3)
      (CallFunction|SetupandUpdateBeam)
      (CallFunction|StartFrameUpdateforPropPreview)
      (else
        _))
    (else
      _))
  (Components|StaticMesh|SetStaticMesh (Variables|Default|GetStaticMeshPreview) 0))

(fn ShouldShowSoundIndicator
  (params (PlayerController :type Fort%20Player%20Controller%20Object%20Reference))
  (return true))

(fn ConstructionScript
  (|CallParentFunction))
