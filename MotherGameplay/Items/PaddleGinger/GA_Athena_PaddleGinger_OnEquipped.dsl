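; Gameplay ability that runs while the PaddleGinger item is equipped, keeping a gameplay effect on the player pawn outside vehicles.
; Activation sends an activate event, applies the effect with a stored handle, and starts AthenaInVehicle tag waits that remove or re-add it.
; Ending prints a debug message, sends a deactivate event, cancels abilities with the PossessProp tag, removes the effect and stops pawn firing.
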(event Custom|AddGE
  ;[1 1 1] Add While Active GE
  (bind _returnvalue (GameplayEffects|ApplyGameplayEffectToSelf :self (Ability|GetAbilitySystemComponent (Variables|Default|GetFort%20Player%20Pawn)) :GameplayEffectClass (Variables|Default|GetGameplay%20Effect%20Class) :Level 0.0))
  (Variables|Default|SetGE_Handle _returnvalue))

(event Custom|RemoveGE
  ;[1 1 1] Remove WhileActive GE
  (GameplayEffects|RemoveActiveGameplayEffect (Ability|GetAbilitySystemComponent (Variables|Default|GetFort%20Player%20Pawn)) (Variables|Default|GetGE_Handle) -1))

(event Ability|EventActivateAbility
  (bind (_event_data _fort_player_pawn) (Setup|SetupGenericAbility true false)
    (:Output
      (Variables|Default|SetFort%20Player%20Pawn _fort_player_pawn)
      (Ability|SendGameplayEventtoActor _fort_player_pawn (Variables|Default|GetEvent%20Tag%20Activate) _event_data)
      ;[1 1 1] Add GE
      (CallFunction|AddGE)
      ;[1 1 1] Handle in/out of vehicles
      (Ability|Tasks|WaitGameplayTagRemove "(TagName=\"Gameplay.Action.Player.AthenaInVehicle\")"
        (:then
          (Ability|Tasks|WaitGameplayTagAdd "(TagName=\"Gameplay.Action.Player.AthenaInVehicle\")"
            (:Added
              (CallFunction|RemoveGE))))
        (:Removed
          (CallFunction|AddGE))))))

(event Ability|EventOnEndAbility
  (params (bWasCancelled :type Boolean))
  (Development|PrintString "PADDLEGINGNER Unequipped" true true "(R=0.000000,G=0.660000,B=1.000000,A=1.000000)" 2e+01)
  (Ability|SendGameplayEventtoActor (stale Ability|EventActivateAbility _fort_player_pawn) (Variables|Default|GetEvent%20Tag%20Deactivate) (stale Ability|EventActivateAbility _event_data))
  (Tags|CancelAbilitiesWithTags (Ability|GetActivatingAbilityComponent) (Variables|Default|GetPossessPropTag))
  ;[1 1 1] Remove GE
  (CallFunction|RemoveGE)
  (Weapon|PawnStopFire (Variables|Default|GetFort%20Player%20Pawn) 0))
