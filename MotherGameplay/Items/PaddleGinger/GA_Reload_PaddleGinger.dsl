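; GA_Reload_PaddleGinger is a gameplay ability that signals a reload for the PaddleGinger item.
; On activation it calls the parent, runs SetupGenericAbility, and sends the reload gameplay event tag to the owning Fort player pawn.
; Releasing the ability input ends the ability, while the reload logic itself is left to whatever handles that event.
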(event Ability|EventActivateAbility
  (|CallParentFunction)
  (bind (_event_data _fort_player_pawn) (Setup|SetupGenericAbility false false)
    (:Output
      (Ability|SendGameplayEventtoActor _fort_player_pawn (Variables|Default|GetEvent%20TagReload)))))

(event Ability|EventOnAbilityInputReleased
  (Ability|EndAbility))
