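; Player prop-disguise actor for the PaddleGinger item that makes a pawn appear as a possessed prop.
; SetProp caches bounds and data-table rotation and scale; a fast timer, guarded per frame, follows the pelvis, or the capsule when placed, lerping rotation.
; Input events drive placing, noise-making and cancelling through ability tags.
; Possession ends when the authority sees another weapon, and entering a vehicle disables prop collision.
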(event EventBeginPlay
  ;[1 1 1] Frame update (basically a tick) to do all the complicated math to calculate our final transform.
  (if (Variables|Default|GetFrameUpdateTransform)
    (Utilities|Time|SetTimerbyEvent (delegate Custom|FrameUpdatePlacement) 0.01 true)))

(event Custom|FrameUpdatePlacement
  (bind (_returnvalue_x _returnvalue_y) (Math|Vector|BreakVector (Math|Vector|RotateVector (Variables|Default|GetDiffToOrigin) (Variables|Default|GetCurrentFrameRot))))
  (bind _returnvalue (Utilities|GetFrameCount))
  (bind _x (Math|Rotator|GetAxes (Variables|Default|GetPrevFrameRot)))
  (bind (_x_x _x_y) (Math|Vector|BreakVector _x))
  (bind (_x_2 _y) (Math|Rotator|GetAxes (Variables|Default|GetCurrentFrameRot)))
  (bind (_y_x _y_y _y_z) (Math|Vector|BreakVector _y))
  (bind _pawnattachedto (Variables|Default|GetPawnAttachedTo))
  (bind _mesh (Variables|/Script/Engine.Character|GetMesh _pawnattachedto))
  (bind (_returnvalue_location _returnvalue_rotation) (Math|Transform|BreakTransform (Transformation|GetSocketTransform _mesh "SM_Pelvis" "RTS_World")))
  (bind (_returnvalue_location_x _returnvalue_location_y) (Math|Vector|BreakVector _returnvalue_location))
  (bind _capsulecomponent (Variables|/Script/Engine.Character|GetCapsuleComponent _pawnattachedto))
  (bind _returnvalue_location_2 (Math|Transform|BreakTransform (Transformation|GetWorldTransform _capsulecomponent)))
  (bind (_returnvalue_location_x_2 _returnvalue_location_y_2 _returnvalue_location_z) (Math|Vector|BreakVector _returnvalue_location_2))
  (bind (_x_3 _y_2 _z) (Math|Rotator|GetAxes _returnvalue_rotation))
  (bind _pawnattachedto_2 (Variables|Default|GetPawnAttachedTo))
  (bind (_boundsextent_x _boundsextent_y) (Math|Vector|BreakVector (Variables|GetBoundsExtent_X)))
  (bind (_currframeloc_x _currframeloc_y _currframeloc_z) (Math|Vector|BreakVector (Variables|GetCurrFrameLoc_X)))
  ;[1 1 1] Frame update (basically a tick) to do all the complicated math to calculate our final transform.
  (if (and (Math|Integer64|NotEqual(Integer64) _returnvalue (Variables|Default|GetLastFrameCount)) (Utilities|IsValid _pawnattachedto_2 :!class /Script/Engine.KismetSystemLibrary))
    (Utilities|FlowControl|Sequence
      (:then_0
        (Variables|Default|SetLastFrameCount _returnvalue)
        (bind _propoffset (CallFunction|CalcPropOffsetPaddleGinger))
        (if (Variables|Default|GetIsPlaced)
          (Variables|Default|SetCurrentFrameRot (Math|Conversions|RotationFromXVector (Math|Vector|Normalize (Math|Vector|MakeVector _x_x _x_y))))
          (bind (_returnvalue_x_2 _returnvalue_y_2 _returnvalue_z) (Math|Vector|BreakVector (Math|Vector|vector+vector (Math|Vector|MakeVector 0.0 0.0 _returnvalue_location_z) _propoffset)))
          (Variables|SetCurrFrameLoc_X (Math|Vector|MakeVector _returnvalue_location_x_2 _returnvalue_location_y_2 _returnvalue_z))
          (else
            (Variables|SetCurrFrameLoc_X (Math|Vector|MakeVector _returnvalue_location_x _returnvalue_location_y _returnvalue_z))
            (Variables|Default|SetCurrentFrameRot (Math|Transform|TransformRotation (Math|Transform|MakeTransform "0, 0, 0" (Math|Rotator|MakeRotfromXZ (* _x_3 -1.0) _z)) (Variables|Default|GetAttachRotRelative)))))
        (if (< (Variables|Default|GetLerpRotAlpha) 1.0)
          ;;[1 1 1] Interpolate the rotation as needed.
          (Variables|Default|SetCurrentFrameRot (Math|Rotator|Lerp(Rotator) (Variables|Default|GetPrevFrameRot) (Variables|Default|GetCurrentFrameRot) (Variables|Default|GetLerpRotAlpha) true))
          (Variables|Default|SetLerpRotAlpha (+ (Variables|Default|GetLerpRotAlpha) (* (Utilities|Time|GetWorldDeltaSeconds) (Variables|Default|GetRotLerpSpeed))))
          (else
            (Variables|Default|SetCurrentFrameRot (Variables|Default|GetCurrentFrameRot))))
        (Variables|SetCurrFrameLoc_X (Math|Vector|MakeVector (+ _currframeloc_x _returnvalue_x) (+ _currframeloc_y _returnvalue_y) (+ _currframeloc_z (Math|Float|SelectFloat 0.0 (+ 0.0 (* (Math|Float|Absolute(Float) _y_z) (* (Math|Float|Max(Float) _boundsextent_x _boundsextent_y) (Variables|Default|GetliftByXExtentsMult)))) (Variables|Default|GetIsPlaced)))))
        (Transformation|SetActorTransform 0 (Math|Transform|MakeTransform (Variables|Default|GetCurrFrameLoc) (Variables|Default|GetCurrentFrameRot) (Variables|Default|GetAttachScaleRelative)) false false)
        (Variables|Default|SetPrevFrameRot (Variables|Default|GetCurrentFrameRot)))
      (:then_1
        (Networking|SwitchHasAuthority
          (:Authority
            (Utilities|Casting|CastToB_Weapon_PaddleGinger_Child (Variables|/Script/FortniteGame.FortPawn|GetCurrentWeapon _pawnattachedto_2)
              (:CastFailed
                (CallFunction|EndPropPossession)))))))))

(event EventDoPawnRelatedInitialization
  (bind _pawnattachedto (Variables|Default|GetPawnAttachedTo))
  ;[1 1 1] DoPawnRelatedInitialization - push input component
  (|CallParentFunction)
  (CallFunction|SetupAndHandlePawnEnteringVehicle)
  (if (Pawn|IsLocallyControlled _pawnattachedto)
    (PropHunt|PushPossessedPropInputMode (Variables|Default|GetFortDataDrivenPossessedPropInput) (Fortnite|GetFortPlayerControllerfromActor _pawnattachedto))))

(event EventEndPropPossession
  (bind _pawnattachedto (Variables|Default|GetPawnAttachedTo))
  ;[1 1 1] EndPropPossession - pop input component
  (Utilities|FlowControl|DoOnce false
    (:Completed
      (Utilities|FlowControl|Sequence
        (:then_0
          (if (Pawn|IsLocallyControlled _pawnattachedto)
            (PropHunt|PopPossessedPropInputMode (Variables|Default|GetFortDataDrivenPossessedPropInput) (Fortnite|GetFortPlayerControllerfromActor _pawnattachedto))))
        (:then_1
          (|CallParentFunction)
          (CallFunction|PlayerStartTransitionBacktoPlayerFXPreChange)
          (GameplayEffects|ApplyGameplayEffectToSelf :self (Ability|GetAbilitySystemComponent (Variables|Default|GetPawnAttachedTo)) :GameplayEffectClass "/MotherGameplay/Items/PaddleGinger/GE_PaddleGinger_PossessPropRemoved.GE_PaddleGinger_PossessPropRemoved_C" :Level 0.0))))))

(event Custom|PlayerStartTransitionBackToPlayerFX_PreChange
  ;[1 1 1] PlayerStartTransitionBackToPlayerFX_PreChange
  (if (not (Variables|Default|GetHasPlayedPreChangeFX))
    (GameplayCue|ExecuteGameplayCue%28Burst%29Local (Variables|Default|GetPawnAttachedTo) (Variables|Default|GetDeactivate_GCNTag) (Ability|GameplayCue|MakeGameplayCueParameters :NormalizedMagnitude 0.0 :RawMagnitude 0.0 :Location (Transformation|GetActorLocation (Variables|Default|GetPawnAttachedTo)) :Normal "0,0,0" :Instigator (Variables|Default|GetPawnAttachedTo) :EffectCauser (Variables|Default|GetPawnAttachedTo) :SourceObject (Variables|Default|GetPawnAttachedTo) :GameplayEffectLevel 1 :AbilityLevel 1 :bReplicateLocationWhenUsingMinimalRepProxy false))
    (Variables|Default|SetHasPlayedPreChangeFX true)))

(event OnPossessedPropCancel(FortDataDrivenPossessedPropInput)
  (bind _pawnattachedto (Variables|Default|GetPawnAttachedTo))
  ;[1 1 1] OnPossessedPropCancel - cancel possessed prop ability
  (CallFunction|PlayerStartTransitionBacktoPlayerFXPreChange)
  (Utilities|FlowControl|Delay 0.1)
  (if (> (Actor|GetGameTimeSinceCreation) (Variables|Default|GetMinPropPossessionTime))
    (Character|UnCrouch _pawnattachedto)
    (bind _asfort_ability_system_component (Utilities|Casting|CastToFortAbilitySystemComponent (Ability|GetAbilitySystemComponent _pawnattachedto)))
    (Tags|CancelAbilitiesWithTags _asfort_ability_system_component (Variables|Default|GetPossessedPropTag))
    (Ability|SendGameplayEventtoActor (Variables|Default|GetPawnAttachedTo) (Variables|Default|GetDeactivate%20Event%20Tag))))

(event OnPossessedPropMakeNoise(FortDataDrivenPossessedPropInput)
  ;[1 1 1] OnPossessedPropMakeNoise - activate "make noise" ability
  (Abilities|TryActivateAbilitiesbyTag (Ability|GetAbilitySystemComponent (Variables|Default|GetPawnAttachedTo)) (Variables|Default|GetPossessedPropMakeNoiseTag) true))

(event OnPossessedPropPlace(FortDataDrivenPossessedPropInput)
  ;[1 1 1] OnPossessedPropPlace - Toggle prop placement mode
  (if (Variables|Default|GetIsPlaced)
    (bind _asfort_ability_system_component (Utilities|Casting|CastToFortAbilitySystemComponent (Ability|GetAbilitySystemComponent (Variables|Default|GetPawnAttachedTo))))
    (Tags|CancelAbilitiesWithTags _asfort_ability_system_component (Variables|Default|GetPossessedPropPlacedTag))
    (Variables|Default|SetLerpRotAlpha 0.0)
    (elif (Variables|Default|GetAllowPropPlacementAbility)
      (Abilities|TryActivateAbilitiesbyTag (Ability|GetAbilitySystemComponent (Variables|Default|GetPawnAttachedTo)) (Variables|Default|GetPossessedPropPlacedTag) true)
      (Variables|Default|SetLerpRotAlpha 0.0))))

(event EventUpdateHealth)

(event EventSetMaxHealth
  (params (NewMaxHealth :type Float%20%28double-precision%29)))

(event EventFinishPropSetup
  ;[1 1 1] Finish prop setup
  (Variables|Visibility|SetbShouldBeHiding true)
  (Variables|Default|SetReceiveDecals_0 false)
  (if (Variables|Default|GetFrameUpdateTransform)
    (CallFunction|AttachPaddleGingerProp)
    (else
      (Transformation|SetActorRelativeLocation "0, 0, 0")
      (Transformation|AddActorWorldOffset (CallFunction|CalcPropOffset))
      (Networking|SwitchHasAuthority
        (:Authority
          (PlayerPropAttachment|AttachProp (Variables|Default|GetCopiedProp)))))))

(event Custom|SetupAndHandlePropDamage
  (params (PropActor :type Actor%20Object%20Reference))
  ;[1 1 1] Setup and Handle prop damage
  (Game|Damage|BindEventtoOnTakeAnyDamage PropActor (delegate Custom|OnPropDamaged)))

(event Custom|OnPropDamaged
  (params (DamagedActor :type Actor%20Object%20Reference) (Damage :type Float%20%28single-precision%29) (DamageType :type Damage%20Type%20Object%20Reference) (InstigatedBy :type Controller%20Object%20Reference) (DamageCauser :type Actor%20Object%20Reference))
  ;[1 1 1] Setup and Handle prop damage
  (Utilities|FlowControl|DelayUntilNextTick)
  (Actor|DestroyActor DamagedActor))

(event Custom|Setup%20and%20Handle%20Pawn%20Entering%20Vehicle
  ;[1 1 1] Setup and handle pawn entering vehicle
  (Vehicle|BindEventtoOnVehicleStateChangedEvent (Variables|Default|GetPawnAttachedTo) (delegate Custom|Vehicle%20State%20Changed)))

(event Custom|Vehicle%20State%20Changed
  (params (FortPlayerPawn :type Fort%20Player%20Pawn%20Object%20Reference) (NewVehicle :type Actor%20Object%20Reference) (OldVehicle :type Actor%20Object%20Reference))
  ;[1 1 1] Setup and handle pawn entering vehicle
  (Development|PrintString (Utilities|String|Append "Disguise Prop detected Vehicle State Changed:" (Utilities|GetDisplayName NewVehicle) "  pawn " (Utilities|GetDisplayName FortPlayerPawn)) (Variables|Default|GetDebugPropInVehicle) (Variables|Default|GetDebugPropInVehicle) "(R=1.000000,G=0.000000,B=0.185009,A=1.000000)" 1e+01)
  (Utilities|IsValid NewVehicle
    (:Is%20Valid
      (Collision|SetActorEnableCollision (Variables|Default|GetCopiedProp) false)
      (Collision|SetActorEnableCollision false))))

(event EventTransitionVFX
  (params (Direction :type ETimelineDirection%20Enum :default "Forward")))

(fn AttachPaddleGingerProp
  (PropHunt|SetPossessedProp (Variables|Default|GetPawnAttachedTo) self)
  ;[1 1 1] Prevent GravyGoblin interaction
  (CallFunction|AddPropstoItemsSpecificBlackList))

(fn CalcPropOffsetPaddleGinger
  (return (Math|Vector|MakeVector 0.0 0.0 (* (Components|Capsule|GetScaledCapsuleHalfHeight (Variables|/Script/Engine.Character|GetCapsuleComponent (Variables|Default|GetPawnAttachedTo))) -1.0))))

(fn CalcPropOffset
  (if (Variables|Default|GetFrameUpdateTransform)
    (return (Math|Vector|MakeVector 0.0 0.0 (+ 0.0 (* (Math|Float|SelectFloat (Components|Capsule|GetScaledCapsuleHalfHeight (Variables|/Script/Engine.Character|GetCapsuleComponent (Variables|Default|GetPawnAttachedTo))) (Components|Capsule|GetScaledCapsuleHalfHeight (Variables|/Script/Engine.Character|GetCapsuleComponent (Variables|Default|GetPawnAttachedTo)))) -1.0))))
    (else
      (return (|CallParentFunction)))))

(fn OnRep_ReceiveDecals_0
  (Default|SetDisableDecals (Variables|Default|GetReceiveDecals_0)))

(fn SetProp
  (params (SelectedProp :type Actor%20Object%20Reference) (PlayerPawnToAttachTo :type Player%20Pawn%20Athena%20Object%20Reference) (ScaleDuration :type Float%20%28double-precision%29))
  (bind _copiedprop (Variables|Default|GetCopiedProp))
  (bind (_origin _extent) (PlayerPropAttachment|CalcWorldSpaceBounds _copiedprop))
  (bind _as_b_weapon_paddle_ginger_child (Variables|Default|GetAs%20B%20Weapon%20Paddle%20Ginger%20Child))
  (bind _propdatatable (Variables|/MotherGameplay/Items/PaddleGinger/B_Weapon_PaddleGinger_Child.B_Weapon_PaddleGinger_Child_C|GetPropDataTable _as_b_weapon_paddle_ginger_child))
  (bind _copiedprop_2 (Variables|Default|GetCopiedProp))
  (if (Variables|Default|GetFrameUpdateTransform)
    (PlayerPropAttachment|InitProp (Variables|GetSelectedProp))
    (Fortnite|SetCanbeDamaged (Variables|GetSelectedProp) true)
    (Networking|SetReplicates (Variables|GetSelectedProp) true)
    (Variables|Default|SetCopiedProp (Variables|GetSelectedProp))
    (bind _output_get (Variables|Default|SetPawnAttachedTo (Variables|GetPlayerPawnToAttachTo)))
    (bind _asb_weapon_paddle_ginger_child (Utilities|Casting|CastToB_Weapon_PaddleGinger_Child (Variables|/Script/FortniteGame.FortPawn|GetCurrentWeapon _output_get))
      (:then
        (Variables|Default|SetAs%20B%20Weapon%20Paddle%20Ginger%20Child _asb_weapon_paddle_ginger_child)
        (bind _outrownames (DataTable|GetDataTableRowNames _propdatatable))
        (bind _returnvalue (Utilities|GetDataTableRow _propdatatable (Utilities|Array|Get(aref) _outrownames (Variables|/MotherGameplay/Items/PaddleGinger/B_Weapon_PaddleGinger_Mother.B_Weapon_PaddleGinger_Mother_C|GetCuratedItemPreviewIndex _as_b_weapon_paddle_ginger_child)))
          (:then
            (Variables|Default|SetAttachRotRelative (.PossessedRotation (Utilities|Struct|BreakPossessPropClasses _returnvalue)))
            (Variables|Default|SetAttachScaleRelative (.PossessedScale (Utilities|Struct|BreakPossessPropClasses _returnvalue))))
          (:RowNotFound
            (Development|PrintString "ERROR: Could not find row for setting relative rot and scale")))
        (Variables|Default|SetDiffToOrigin (Math|Vector|UnrotateVector (- (Transformation|GetActorLocation _copiedprop) _origin) (Transformation|GetActorRotation)))
        (Variables|Default|SetBoundsExtent (Math|Vector|MakeVector (Math|Float|Clamp(Float) (.x _extent) (Variables|PlayerSize|GetMinCapsuleRadius) (Variables|PlayerSize|GetMaxCapsuleRadius)) (Math|Float|Clamp(Float) (.y _extent) (Variables|PlayerSize|GetMinCapsuleRadius) (Variables|PlayerSize|GetMaxCapsuleRadius)) (Math|Float|Clamp(Float) (.z _extent) (Variables|PlayerSize|GetMinCapsuleHalfHeight) (Variables|PlayerSize|GetMaxCapsuleHalfHeight))))
        (Actor|SetOwner (Variables|Default|GetCopiedProp) self)
        (CallFunction|MulticastSetStaticMesh (Variables|Default|GetPawnAttachedTo) (Variables|GetScaleDuration) (Variables|Default|GetBoundsExtent))
        (PlayerPropAttachment|LogState "Post BP - SetProp (Inc Attach To Pawn)")
        (Fortnite|ApplyGameplayEffectsToActor (Variables|GetSelectedProp) (Variables|Default|GetGE_ForPropMaxHealth) (Transformation|GetActorLocation (Variables|GetSelectedProp)) false 0 (Ability|GetAbilitySystemComponent (Variables|Default|GetPawnAttachedTo)))
        (for _array_element (Actor|GetComponentsByClass _copiedprop_2 "/Script/Engine.StaticMeshComponent")
          (Rendering|SetReceivesDecals _array_element false))
        (for _array_element_2 (Actor|GetComponentsByClass _copiedprop_2 "/Script/Engine.SkeletalMeshComponent")
          (Rendering|SetReceivesDecals _array_element_2 false)))
      (:CastFailed
        (Development|PrintString "ERROR Pawn's weapon is not a paddle chinger")))))

(fn SetPropPlaced
  (params (NewPlaced :type Boolean) (PlacedYawOffset :type Float%20%28double-precision%29))
  (bind _returnvalue (BlueprintContext|GetCreativeHUDLContext))
  (if (Variables|Default|GetFrameUpdateTransform)
    (bind _output_get (Variables|Default|SetIsPlaced NewPlaced))
    (Utilities|FlowControl|Sequence
      (:then_0
        (Utilities|IsValid _returnvalue
          (:Is%20Valid
            (Class|CreativeHUDLContext|ForwardonPropLockedinPlacedChanged _returnvalue _output_get))))
      (:then_1
        (if _output_get
          (bind _output_get_2 (Variables|Default|SetPrePlacementRelativeTransform (Transformation|GetRelativeTransform (Variables|/Script/Engine.Actor|GetRootComponent))))
          (Variables|Default|SetPrePlacementRotation (.rotation _output_get_2))
          ;[1 1 1] Since the server does not play animations, our prop rotation will always be offset from what clients see at the time of placement. We send the PlacedYawOffset from the locally controlled client so clients don't experience a rotation pop when entering "placed" mode
          (if (Networking|IsServer)
            (Transformation|AddWorldRotation (Variables|Transformation|GetRootComponent) (Math|Rotator|MakeRotator 0.0 0.0 (Variables|GetPlacedYawOffset)) false false))
          (else
            _))))))

(fn ConstructionScript
  (|CallParentFunction))
