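# A small UserWidget that draws a single pip indicator for the PaddleGinger prop, showing whether it is selected.
# Its only element is an Image using the M_UI_PaddleGinger_Pip material.
# SetState takes a Selected flag and writes 1 or 0 to the IsSelected scalar on the image's dynamic material, so the material handles the visual change.
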WidgetBlueprint /MotherGameplay/Items/PaddleGinger/UI/PaddleGinger_Prop_Pip_Indicator
graphs [
 EventGraph {links [], nodes []}
 SetState {nodes [
   SetState!() -> (then, Selected:bool=~) {is=title K2Node_FunctionEntry}
   > n1 = GetDynamicMaterial(self=Get Pip_Image() -> (Pip_Image:Image Object Reference))
   > SetScalarParameterValue(
    self=@n1
    ParameterName=IsSelected
    Value=Select(Option 0:double=0, Option 1:double=1, Index=Get Selected() -> (Selected:bool)) -> (ReturnValue:double)
   )
  ]}
]
variables []
widgets {
 info {inheritedWidgetCount=0, parentClass=&UserWidget, rootWidgetClass=&Image, widgetCount=1}
 widgets [Pip_Image {
   widgetClassPath=&Image
   bIsVariable=true
   widget=Pip_Image
   properties {brush {
     bIsDynamicallyLoaded=false
     drawAs=Image
     imageType=NoImage
     mirroring=NoMirror
     resourceName=None
     resourceObject=&/M_UI_PaddleGinger_Pip
     tiling=NoTile
     imageSize=<20 20>
     margin=<ltrb 0 0 0 0>
     outlineSettings {
      bUseBrushTransparency=false
      roundingType=HalfHeightRadius
      width=0
      color {colorUseRule=UseColor_Specified, specifiedColor=<rgba 0 0 0 0>}
      cornerRadii=<0 0 0 0>
     }
     tintColor {colorUseRule=UseColor_Specified, specifiedColor=<rgba 1 1 1 1>}
     uVRegion {bIsValid=false, max=<0 0>, min=<0 0>}
    }}
  }]
}
