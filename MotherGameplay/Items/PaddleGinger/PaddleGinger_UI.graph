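# Reticle extension HUD for the PaddleGinger item, showing prop category and prop selection with keybind prompts for cycling and activating.
# OnPropIndexChanged updates category and one-based counter texts, then SetPropPips clears and recreates one pip widget per prop, marking the selected one.
# OnPropActivationChanged collapses the widget while activated and shows it otherwise.
# Construct enlarges prompts on mobile, plays a flickering slide-in Intro with animated material parameters, and starts visible.
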WidgetBlueprint /MotherGameplay/Items/PaddleGinger/PaddleGinger_UI
animations [Intro {
  binding_count=11
  length_seconds=0.55
  object=&self:Intro
  sequence {bindings [
    RightSide {
     id=#1
     possessed_class=HorizontalBox
     tracks [
      {
       type=MovieSceneFloatTrack
       display_name=Render Opacity
       sections [{type=MovieSceneFloatSection, channels [None {type=MovieSceneScriptingFloatChannel, keys [[3,0],[4,0.5],[5,0.125],[6,0.75],[7,0.125],[9,1]]}]}]
      }
      {
       type=MovieScene2DTransformTrack
       sections [{
         type=MovieScene2DTransformSection
         channels [
          Translation.X {type=MovieSceneScriptingFloatChannel, keys [[3,-64],[7,0]]}
          Translation.Y {type=MovieSceneScriptingFloatChannel, const [0,1]}
          Angle {type=MovieSceneScriptingFloatChannel, const [0,1]}
          Scale.X {type=MovieSceneScriptingFloatChannel, const [1,1]}
          Scale.Y {type=MovieSceneScriptingFloatChannel, const [1,1]}
          Shear.X {type=MovieSceneScriptingFloatChannel, const [0,1]}
          Shear.Y {type=MovieSceneScriptingFloatChannel, const [0,1]}
         ]
        }]
      }
     ]
    }
    RightLine {
     id=#2
     possessed_class=Image
     tracks [
      {
       type=MovieSceneWidgetMaterialTrack
       sections [{type=MovieSceneParameterSection, channels [Mask_Offset {type=MovieSceneScriptingFloatChannel, keys [[6,-0.965],[11,-0.9275]]}]}]
      }
      {
       type=MovieScene2DTransformTrack
       sections [{
         type=MovieScene2DTransformSection
         channels [
          Translation.X {type=MovieSceneScriptingFloatChannel, keys [[3,-8],[11,0]]}
          Translation.Y {type=MovieSceneScriptingFloatChannel, const [0,1]}
          Angle {type=MovieSceneScriptingFloatChannel, const [0,1]}
          Scale.X {type=MovieSceneScriptingFloatChannel, const [-1,1]}
          Scale.Y {type=MovieSceneScriptingFloatChannel, keys [[3,0.25],[6,1]]}
          Shear.X {type=MovieSceneScriptingFloatChannel, const [-2,1]}
          Shear.Y {type=MovieSceneScriptingFloatChannel, const [1,1]}
         ]
        }]
      }
     ]
    }
    PropIndicatorBox {
     id=#3
     possessed_class=HorizontalBox
     tracks [{
       type=MovieScene2DTransformTrack
       sections [{
         type=MovieScene2DTransformSection
         channels [
          Translation.X {type=MovieSceneScriptingFloatChannel, keys [[3,-16],[9,0]]}
          Translation.Y {type=MovieSceneScriptingFloatChannel}
         ]
        }]
      }]
    }
    LeftSide {
     id=#4
     possessed_class=HorizontalBox
     tracks [
      {
       type=MovieSceneFloatTrack
       display_name=Render Opacity
       sections [{type=MovieSceneFloatSection, channels [None {type=MovieSceneScriptingFloatChannel, keys [[3,0],[4,0.5],[5,0.125],[6,0.75],[7,0.125],[9,1]]}]}]
      }
      {
       type=MovieScene2DTransformTrack
       sections [{
         type=MovieScene2DTransformSection
         channels [
          Translation.X {type=MovieSceneScriptingFloatChannel, keys [[3,64],[7,0]]}
          Translation.Y {type=MovieSceneScriptingFloatChannel, const [0,1]}
          Angle {type=MovieSceneScriptingFloatChannel, const [0,1]}
          Scale.X {type=MovieSceneScriptingFloatChannel, const [1,1]}
          Scale.Y {type=MovieSceneScriptingFloatChannel, const [1,1]}
          Shear.X {type=MovieSceneScriptingFloatChannel, const [0,1]}
          Shear.Y {type=MovieSceneScriptingFloatChannel, const [0,1]}
         ]
        }]
      }
     ]
    }
    Activate {
     id=#5
     display_name=HorizontalBox_0
     possessed_class=Overlay
     tracks [{
       type=MovieSceneFloatTrack
       display_name=Render Opacity
       sections [{
         type=MovieSceneFloatSection
         channels [None {type=MovieSceneScriptingFloatChannel, keys [[3,1],[6,0],[7,0.5],[8,0.125],[9,0.75],[10,0.125],[11,1]]}]
        }]
      }]
    }
    LeftLine {
     id=#6
     possessed_class=Image
     tracks [
      {
       type=MovieSceneWidgetMaterialTrack
       sections [{type=MovieSceneParameterSection, channels [Mask_Offset {type=MovieSceneScriptingFloatChannel, keys [[6,-0.965],[11,-0.9275]]}]}]
      }
      {
       type=MovieScene2DTransformTrack
       sections [{
         type=MovieScene2DTransformSection
         channels [
          Translation.X {type=MovieSceneScriptingFloatChannel, keys [[3,8],[11,0]]}
          Translation.Y {type=MovieSceneScriptingFloatChannel, const [0,1]}
          Angle {type=MovieSceneScriptingFloatChannel, const [0,1]}
          Scale.X {type=MovieSceneScriptingFloatChannel, const [1,1]}
          Scale.Y {type=MovieSceneScriptingFloatChannel, keys [[3,0.25],[6,1]]}
          Shear.X {type=MovieSceneScriptingFloatChannel, const [2,1]}
          Shear.Y {type=MovieSceneScriptingFloatChannel, const [-1,1]}
         ]
        }]
      }
     ]
    }
    Image_269 {
     id=#7
     possessed_class=Image
     tracks [{
       type=MovieSceneWidgetMaterialTrack
       sections [{type=MovieSceneParameterSection, channels [BackingAmount {type=MovieSceneScriptingFloatChannel, keys [[3,0],[6,0],[11,0.5]]}]}]
      }]
    }
    CategoryIndicatorBox {
     id=#8
     possessed_class=HorizontalBox
     tracks [{
       type=MovieScene2DTransformTrack
       sections [{
         type=MovieScene2DTransformSection
         channels [
          Translation.X {type=MovieSceneScriptingFloatChannel, keys [[3,16],[9,0]]}
          Translation.Y {type=MovieSceneScriptingFloatChannel}
         ]
        }]
      }]
    }
    PropIndicator {
     id=#9
     possessed_class=HorizontalBox
     tracks [{
       type=MovieScene2DTransformTrack
       sections [{
         type=MovieScene2DTransformSection
         channels [
          Translation.X {type=MovieSceneScriptingFloatChannel, keys [[5,-16],[11,0]]}
          Translation.Y {type=MovieSceneScriptingFloatChannel}
         ]
        }]
      }]
    }
    CategoryBox {
     id=#10
     possessed_class=HorizontalBox
     tracks [{
       type=MovieScene2DTransformTrack
       sections [{
         type=MovieScene2DTransformSection
         channels [
          Translation.X {type=MovieSceneScriptingFloatChannel, keys [[5,16],[11,0]]}
          Translation.Y {type=MovieSceneScriptingFloatChannel}
         ]
        }]
      }]
    }
    Activate {
     id=#11
     possessed_class=Overlay
     tracks [
      {
       type=MovieSceneFloatTrack
       display_name=Render Opacity
       sections [{type=MovieSceneFloatSection, channels [None {type=MovieSceneScriptingFloatChannel, keys [[0,0],[6,0],[8,1]]}]}]
      }
      {
       type=MovieScene2DTransformTrack
       sections [{
         type=MovieScene2DTransformSection
         channels [
          Translation.X {type=MovieSceneScriptingFloatChannel, keys [[6,216],[11,232]]}
          Translation.Y {type=MovieSceneScriptingFloatChannel, const [0,1]}
          Angle {type=MovieSceneScriptingFloatChannel, const [0,1]}
          Scale.X {type=MovieSceneScriptingFloatChannel, const [1,1]}
          Scale.Y {type=MovieSceneScriptingFloatChannel, const [1,1]}
          Shear.X {type=MovieSceneScriptingFloatChannel, const [0,1]}
          Shear.Y {type=MovieSceneScriptingFloatChannel, const [0,1]}
         ]
        }]
      }
     ]
    }
   ]}
 }]
graphs [
 AttemptChangeVisibility {nodes [
   n1 = AttemptChangeVisibility!() -> (then, InActivated:bool=~) {is=title K2Node_FunctionEntry}
   > n2 = Branch(Condition=@n1)
   > SetVisibility(InVisibility=Collapsed)
   SetVisibility(execute=@n2.else, InVisibility=HitTestInvisible)
  ]}
 EventGraph {nodes [
   n1 = Event OnPropIndexChanged!() -> (OutputDelegate, then, CategoryText:"Text (by ref)", PropIndex:int=0, MaxPropCount:int=0, CategoryIndex:int=0, MaxCategoryCount:int=0) {is=title K2Node_Event}
   > "SetText (Text)"(self=Get Category_Text() -> (Category_Text:Text Object Reference), InText=@n1.CategoryText)
   > "SetText (Text)"(
    self=Get Index_Text() -> (Index_Text:Text Object Reference)
    InText=Format Text!(
     Format="LOCTEXT(\"({0} of {1})\")"
     0:int=int + int(A=@n1.PropIndex, B=1) -> (ReturnValue=0) {is=title K2Node_CommutativeAssociativeBinaryOperator}
     1:int=@n1.MaxPropCount
    ) -> (Result)
   )
   > "SetText (Text)"(
    self=Get Category_Index_Text() -> (Category_Index_Text:Text Object Reference)
    InText=Format Text(Format="LOCTEXT(\"({0} of {1})\")", 0=int + int(A=@n1.CategoryIndex, B=1) -> (ReturnValue=0), 1=@n1.MaxCategoryCount)
   )
   > SetPropPips!(execute, self, SelectedIndex:int=@n1.PropIndex, MaxCount:int=@n1.MaxPropCount) -> (then)
   n2 = Event OnPropActivationChanged!() -> (OutputDelegate, then, Activated:bool) {is=title K2Node_Event}
   > AttemptChangeVisibility!(execute, self, InActivated:bool=@n2.Activated) -> (then) {is=title K2Node_CallFunction}
   Event Construct()
   > n3 = Branch(Condition=IsMobileGame())
   > n4 = SetRenderScale!(
    execute
    self:Widget Object Reference=[
     Get Interactions() -> (Interactions:Horizontal Box Object Reference)
     Get MobileVisibility() -> (MobileVisibility:Common Visibility Widget Legacy Object Reference)
    ]
    Scale:Vector 2D Structure=(X=1.2,Y=1.2)
   ) -> (then)
   PlayAnimation(execute=[@n4.then, @n3.else], InAnimation=Get Intro())
   > AttemptChangeVisibility(InActivated=false)
  ]}
 SetPropPips {nodes [
   SetPropPips!() -> (then, SelectedIndex:int, MaxCount:int) {is=title K2Node_FunctionEntry}
   > ClearChildren(self=Get PropPips() -> (PropPips:Horizontal Box Object Reference))
   > n1 = For Loop(LastIndex=int - int(A=Get MaxCount() -> (MaxCount:int=0), B=1) {is=title K2Node_CallFunction})
   > n2 = Create Widget!(
    execute:Exec
    Class:User Widget Class Reference=/MotherGameplay/Items/PaddleGinger/UI/PaddleGinger_Prop_Pip_Indicator.PaddleGinger_Prop_Pip_Indicator_C
    OwningPlayer:Player Controller Object Reference=GetOwningPlayer()
   ) -> (then:Exec, ReturnValue:Paddle Ginger Prop Pip Indicator Object Reference)
   > SetState!(
    execute
    self:Paddle Ginger Prop Pip Indicator Object Reference=@n2
    Selected:bool="Equal (Integer)"(A=@n1, B=Get SelectedIndex() -> (SelectedIndex:int=0)) {is=title K2Node_CallFunction}
   ) -> (then)
   > n3 = AddChildToHorizontalBox!(execute, self:Horizontal Box Object Reference=Get PropPips(), Content:Widget Object Reference=@n2) -> (then, ReturnValue:Horizontal Box Slot Object Reference)
   > SetPadding!(execute, self:Horizontal Box Slot Object Reference=@n3, InPadding:Margin Structure=Make Margin(Left=0, Top=0, Right=6, Bottom=0)) -> (then)
  ]}
]
variables [
 Category Index {type=integer}
 Max Category Count {type=integer}
 Max Prop Count {type=integer, default=3}
 Prop Index {type=integer}
]
widgets {
 info {
  inheritedWidgetCount=4
  parentClass=/Script/MotherGameplayUI.FortPaddleGingerReticleExtensionWidget
  rootWidgetClass=&Overlay
  widgetCount=44
 }
 widgets [
  Overlay_0 {widgetClassPath=&Overlay, widget=Overlay_0, properties {renderOpacity=0.9}}
  Size_All {
   widgetClassPath=&SizeBox
   parent=Overlay_0
   widget=Size_All
   properties {
    bOverride_HeightOverride=true
    heightOverride=400
    widthOverride=1024
    renderTransform {angle=0, scale=<1 1>, shear=<0 0>, translation=<0 -16>}
   }
   slot {class=OverlaySlot, properties {horizontalAlignment=HAlign_Center, verticalAlignment=VAlign_Center}}
  }
  VertBox_All {widgetClassPath=&VerticalBox, parent=Size_All, widget=VertBox_All, slot {class=SizeBoxSlot}}
  Interactions {
   widgetClassPath=&HorizontalBox
   bIsVariable=true
   parent=VertBox_All
   widget=Interactions
   slot {class=VerticalBoxSlot, properties {size {sizeRule=Fill, value=1}}}
  }
  SizeBox_LeftSide {
   widgetClassPath=&SizeBox
   parent=Interactions
   widget=SizeBox_LeftSide
   properties {bOverride_WidthOverride=true, widthOverride=384}
   slot {class=HorizontalBoxSlot, properties {horizontalAlignment=HAlign_Left, verticalAlignment=VAlign_Top}}
  }
  LeftSide {
   widgetClassPath=&HorizontalBox
   parent=SizeBox_LeftSide
   widget=LeftSide
   slot {class=SizeBoxSlot, properties {horizontalAlignment=HAlign_Right}}
  }
  LeftSideInfo {
   widgetClassPath=&VerticalBox
   parent=LeftSide
   widget=LeftSideInfo
   properties {renderTransform {angle=0, scale=<1 1>, shear=<4 -2>, translation=<0 8>}}
   slot {class=HorizontalBoxSlot, properties {horizontalAlignment=HAlign_Right, verticalAlignment=VAlign_Center}}
  }
  CategoryIndicatorBox {
   widgetClassPath=&HorizontalBox
   parent=LeftSideInfo
   widget=CategoryIndicatorBox
   slot {class=VerticalBoxSlot, properties {horizontalAlignment=HAlign_Right}}
  }
  RichText_CycleCategoryAction {
   widgetClassPath=&CommonRichTextBlock
   bIsVariable=true
   parent=CategoryIndicatorBox
   widget=RichText_CycleCategoryAction
   properties {
    bOverrideDefaultStyle=true
    text=Category
    textTransformPolicy=ToUpper
    defaultTextStyleOverride {
     overflowPolicy=Clip
     transformPolicy=None
     colorAndOpacity {colorUseRule=UseColor_Specified, specifiedColor=<rgba 0.25 1 1 1>}
     font {
      bForceMonospaced=false
      bMaterialIsStencil=false
      fontMaterial=null
      fontObject=&BurbankSmall
      letterSpacing=0
      monospacedWidth=1
      size=15
      skewAmount=0
      typefaceFontName=Black
      outlineSettings {
       bApplyOutlineToDropShadows=false
       bMiteredCorners=false
       bSeparateFillAlpha=false
       outlineMaterial=&M-FontShadow
       outlineSize=2
       outlineColor=<rgba 0 0.1667 1 0.75>
      }
     }
     highlightColor {colorUseRule=UseColor_Specified, specifiedColor=<rgba 0 0 0 1>}
     highlightShape {
      bIsDynamicallyLoaded=false
      drawAs=Image
      imageType=NoImage
      mirroring=NoMirror
      resourceName=None
      resourceObject=null
      tiling=NoTile
      imageSize=<32 32>
      margin=<ltrb 0 0 0 0>
      outlineSettings {
       bUseBrushTransparency=false
       roundingType=HalfHeightRadius
       width=0
       color {colorUseRule=UseColor_Specified, specifiedColor=<rgba 0 0 0 0>}
       cornerRadii=<0 0 0 0>
      }
      tintColor {colorUseRule=UseColor_Specified, specifiedColor=<rgba 1 1 1 1>}
      uVRegion {bIsValid=false, max=<0 0>, min=<0 0>}
     }
     selectedBackgroundColor {colorUseRule=UseColor_Foreground, specifiedColor=<rgba 1 0 1 1>}
     shadowColorAndOpacity=<rgba 0 0 0 1>
     shadowOffset=<0 0>
     strikeBrush {
      bIsDynamicallyLoaded=false
      drawAs=Image
      imageType=NoImage
      mirroring=NoMirror
      resourceName=None
      resourceObject=null
      tiling=NoTile
      imageSize=<32 32>
      margin=<ltrb 0 0 0 0>
      outlineSettings {
       bUseBrushTransparency=false
       roundingType=HalfHeightRadius
       width=0
       color {colorUseRule=UseColor_Specified, specifiedColor=<rgba 0 0 0 0>}
       cornerRadii=<0 0 0 0>
      }
      tintColor {colorUseRule=UseColor_Specified, specifiedColor=<rgba 1 1 1 1>}
      uVRegion {bIsValid=false, max=<0 0>, min=<0 0>}
     }
     underlineBrush {
      bIsDynamicallyLoaded=false
      drawAs=Image
      imageType=NoImage
      mirroring=NoMirror
      resourceName=None
      resourceObject=null
      tiling=NoTile
      imageSize=<32 32>
      margin=<ltrb 0 0 0 0>
      outlineSettings {
       bUseBrushTransparency=false
       roundingType=HalfHeightRadius
       width=0
       color {colorUseRule=UseColor_Specified, specifiedColor=<rgba 0 0 0 0>}
       cornerRadii=<0 0 0 0>
      }
      tintColor {colorUseRule=UseColor_Specified, specifiedColor=<rgba 1 1 1 1>}
      uVRegion {bIsValid=false, max=<0 0>, min=<0 0>}
     }
    }
    renderTransform {angle=0, scale=<1 1>, shear=<-12 0>, translation=<0 0>}
   }
   slot {class=HorizontalBoxSlot, properties {horizontalAlignment=HAlign_Right, verticalAlignment=VAlign_Center, padding=<ltrb 8 0 0 0>}}
  }
  Category_Index_Text {
   widgetClassPath=&TextBlock
   bIsVariable=true
   parent=CategoryIndicatorBox
   widget=Category_Index_Text
   properties {
    text="{Category Index}"
    visibility=SelfHitTestInvisible
    colorAndOpacity {colorUseRule=UseColor_Specified, specifiedColor=<rgba 0.5 1 1 1>}
    font {
     bForceMonospaced=false
     bMaterialIsStencil=false
     fontMaterial=null
     fontObject=&Burbank
     letterSpacing=0
     monospacedWidth=1
     size=13
     skewAmount=0
     typefaceFontName=Burbank-PlayerPower
     outlineSettings {
      bApplyOutlineToDropShadows=true
      bMiteredCorners=false
      bSeparateFillAlpha=true
      outlineMaterial=&M-FontShadow
      outlineSize=2
      outlineColor=<rgba 0 0.1667 1 0.8>
     }
    }
    renderTransform {angle=0, scale=<1 1>, shear=<-12 0>, translation=<0 0>}
   }
   slot {class=HorizontalBoxSlot, properties {horizontalAlignment=HAlign_Right, verticalAlignment=VAlign_Bottom, padding=<ltrb 4 0 0 2>}}
  }
  SizeBox_0 {
   widgetClassPath=&SizeBox
   parent=LeftSideInfo
   widget=SizeBox_0
   properties {bOverride_WidthOverride=true, widthOverride=256}
   slot {class=VerticalBoxSlot, properties {horizontalAlignment=HAlign_Right}}
  }
  CategoryBox {
   widgetClassPath=&HorizontalBox
   parent=SizeBox_0
   widget=CategoryBox
   slot {class=SizeBoxSlot, properties {horizontalAlignment=HAlign_Right}}
  }
  LeftArrow {
   widgetClassPath=&Image
   bIsVariable=true
   parent=CategoryBox
   widget=LeftArrow
   properties {brush {
     bIsDynamicallyLoaded=false
     drawAs=Image
     imageType=NoImage
     mirroring=NoMirror
     resourceName=None
     resourceObject=/MotherGameplay/Items/PaddleGinger/UI/M_UI_PaddleGinger_Glitch_Triangle
     tiling=NoTile
     imageSize=<30 30>
     margin=<ltrb 0 0 0 0>
     outlineSettings {
      bUseBrushTransparency=false
      roundingType=HalfHeightRadius
      width=0
      color {colorUseRule=UseColor_Specified, specifiedColor=<rgba 0 0 0 0>}
      cornerRadii=<0 0 0 0>
     }
     tintColor {colorUseRule=UseColor_Specified, specifiedColor=<rgba 1 1 1 1>}
     uVRegion {bIsValid=false, max=<0 0>, min=<0 0>}
    }}
   slot {class=HorizontalBoxSlot, properties {horizontalAlignment=HAlign_Right, verticalAlignment=VAlign_Center}}
  }
  Category_Text {
   widgetClassPath=&TextBlock
   bIsVariable=true
   parent=CategoryBox
   widget=Category_Text
   properties {
    justification=Right
    text="{Category}"
    textTransformPolicy=ToUpper
    visibility=SelfHitTestInvisible
    font {
     bForceMonospaced=false
     bMaterialIsStencil=false
     fontMaterial=null
     fontObject=&Burbank
     letterSpacing=0
     monospacedWidth=1
     size=18
     skewAmount=0
     typefaceFontName=BurbankRegular-Black
     outlineSettings {
      bApplyOutlineToDropShadows=true
      bMiteredCorners=false
      bSeparateFillAlpha=true
      outlineMaterial=&M-FontShadow
      outlineSize=3
      outlineColor=<rgba 0 0.1667 1 0.75>
     }
    }
    renderTransform {angle=0, scale=<1 1>, shear=<-12 0>, translation=<0 0>}
   }
   slot {class=HorizontalBoxSlot, properties {horizontalAlignment=HAlign_Right}}
  }
  LeftLine {
   widgetClassPath=&Image
   bIsVariable=true
   parent=LeftSide
   widget=LeftLine
   properties {
    visibility=SelfHitTestInvisible
    brush {
     bIsDynamicallyLoaded=false
     drawAs=Image
     imageType=NoImage
     mirroring=NoMirror
     resourceName=None
     resourceObject=/MotherGameplay/Items/PaddleGinger/UI/M_UI_PaddleGinger_ThinLine
     tiling=NoTile
     imageSize=<104 104>
     margin=<ltrb 0.5 0 0.5 0>
     outlineSettings {
      bUseBrushTransparency=false
      roundingType=HalfHeightRadius
      width=0
      color {colorUseRule=UseColor_Specified, specifiedColor=<rgba 0 0 0 0>}
      cornerRadii=<0 0 0 0>
     }
     tintColor {colorUseRule=UseColor_Specified, specifiedColor=<rgba 1 1 1 1>}
     uVRegion {bIsValid=false, max=<0 0>, min=<0 0>}
    }
    renderTransform {angle=0, scale=<1 1>, shear=<2 -1>, translation=<0 0>}
   }
   slot {class=HorizontalBoxSlot, properties {horizontalAlignment=HAlign_Right, padding=<ltrb -88 0 0 0>}}
  }
  MobileIcon_CycleCategory {
   widgetClassPath=&CommonVisibilityWidgetLegacy
   parent=LeftSide
   widget=MobileIcon_CycleCategory
   properties {
    bShowForAndroid=false
    bShowForErebus=false
    bShowForGamepad=false
    bShowForIOS=false
    bShowForMac=false
    bShowForMouseAndKeyboard=false
    bShowForPC=false
    bShowForPS4=false
    bShowForPS5=false
    bShowForSwitch2=false
    bShowForXBox=false
    bShowForXSX=false
    padding=<ltrb 8 0 0 0>
   }
   slot {class=HorizontalBoxSlot, properties {padding=<ltrb 0 0 -40 0>}}
  }
  ScaleBox {
   widgetClassPath=&ScaleBox
   parent=MobileIcon_CycleCategory
   widget=ScaleBox
   properties {stretch=UserSpecified, userSpecifiedScale=0.8}
   slot {class=BorderSlot, properties {padding=<ltrb 8 0 0 0>}}
  }
  Mobile_CycleCategory {
   widgetClassPath=&Image
   bIsVariable=true
   parent=ScaleBox
   widget=Mobile_CycleCategory
   properties {
    visibility=SelfHitTestInvisible
    brush {
     bIsDynamicallyLoaded=false
     drawAs=Image
     imageType=NoImage
     mirroring=NoMirror
     resourceName=None
     resourceObject=/MotherGameplay/Items/PaddleGinger/HUD/Icon_CyclePropCategory_Sprite
     tiling=NoTile
     imageSize=<88 72>
     margin=<ltrb 0 0 0 0>
     outlineSettings {
      bUseBrushTransparency=false
      roundingType=HalfHeightRadius
      width=0
      color {colorUseRule=UseColor_Specified, specifiedColor=<rgba 0 0 0 0>}
      cornerRadii=<0 0 0 0>
     }
     tintColor {colorUseRule=UseColor_Specified, specifiedColor=<rgba 1 1 1 1>}
     uVRegion {bIsValid=false, max=<0 0>, min=<0 0>}
    }
    colorAndOpacity=<rgba 0.5 1 1 0.8>
   }
   slot {class=ScaleBoxSlot}
  }
  ScaleBox_0 {
   widgetClassPath=&ScaleBox
   parent=LeftSide
   widget=ScaleBox_0
   properties {stretch=UserSpecified, userSpecifiedScale=0.9}
   slot {class=HorizontalBoxSlot, properties {horizontalAlignment=HAlign_Center, verticalAlignment=VAlign_Center, padding=<ltrb 8 0 0 0>}}
  }
  Keybind_CycleCategoryAction {
   widgetClassPath=/FortUILibrary/Input/KeybindWidget
   bInherited=true
   bIsVariable=true
   parent=ScaleBox_0
   widget=Keybind_CycleCategoryAction
   slot {class=ScaleBoxSlot}
  }
  Spacer_98 {widgetClassPath=&Spacer, parent=Interactions, widget=Spacer_98, properties {size=<336 1>}, slot {class=HorizontalBoxSlot}}
  SizeBox_RightSide {
   widgetClassPath=&SizeBox
   parent=Interactions
   widget=SizeBox_RightSide
   properties {bOverride_WidthOverride=true, widthOverride=384}
   slot {class=HorizontalBoxSlot, properties {horizontalAlignment=HAlign_Right, verticalAlignment=VAlign_Top, size {sizeRule=Fill, value=1}}}
  }
  RightSide {
   widgetClassPath=&HorizontalBox
   parent=SizeBox_RightSide
   widget=RightSide
   slot {class=SizeBoxSlot, properties {horizontalAlignment=HAlign_Left}}
  }
  ScaleBox_1 {
   widgetClassPath=&ScaleBox
   parent=RightSide
   widget=ScaleBox_1
   properties {stretch=UserSpecified, userSpecifiedScale=0.9}
   slot {class=HorizontalBoxSlot, properties {horizontalAlignment=HAlign_Center, verticalAlignment=VAlign_Center, padding=<ltrb 0 0 8 0>}}
  }
  Keybind_CyclePropAction {
   widgetClassPath=&KeybindWidget
   bInherited=true
   bIsVariable=true
   parent=ScaleBox_1
   widget=Keybind_CyclePropAction
   slot {class=ScaleBoxSlot}
  }
  MobileIcon_CycleProp {
   widgetClassPath=&CommonVisibilityWidgetLegacy
   parent=RightSide
   widget=MobileIcon_CycleProp
   properties {
    bShowForAndroid=false
    bShowForErebus=false
    bShowForGamepad=false
    bShowForIOS=false
    bShowForMac=false
    bShowForMouseAndKeyboard=false
    bShowForPC=false
    bShowForPS4=false
    bShowForPS5=false
    bShowForSwitch2=false
    bShowForXBox=false
    bShowForXSX=false
    padding=<ltrb 0 0 8 0>
   }
   slot {class=HorizontalBoxSlot, properties {padding=<ltrb -40 0 0 0>}}
  }
  ScaleBox_4 {
   widgetClassPath=&ScaleBox
   parent=MobileIcon_CycleProp
   widget=ScaleBox_4
   properties {stretch=UserSpecified, userSpecifiedScale=0.8}
   slot {class=BorderSlot, properties {padding=<ltrb 0 0 8 0>}}
  }
  Mobile_CycleProp {
   widgetClassPath=&Image
   bIsVariable=true
   parent=ScaleBox_4
   widget=Mobile_CycleProp
   properties {
    visibility=SelfHitTestInvisible
    brush {
     bIsDynamicallyLoaded=false
     drawAs=Image
     imageType=NoImage
     mirroring=NoMirror
     resourceName=None
     resourceObject=/MotherGameplay/Items/PaddleGinger/HUD/Icon_CycleProp_Sprite
     tiling=NoTile
     imageSize=<88 72>
     margin=<ltrb 0 0 0 0>
     outlineSettings {
      bUseBrushTransparency=false
      roundingType=HalfHeightRadius
      width=0
      color {colorUseRule=UseColor_Specified, specifiedColor=<rgba 0 0 0 0>}
      cornerRadii=<0 0 0 0>
     }
     tintColor {colorUseRule=UseColor_Specified, specifiedColor=<rgba 1 1 1 1>}
     uVRegion {bIsValid=false, max=<0 0>, min=<0 0>}
    }
    colorAndOpacity=<rgba 0.5 1 1 0.8>
   }
   slot {class=ScaleBoxSlot}
  }
  RightLine {
   widgetClassPath=&Image
   bIsVariable=true
   parent=RightSide
   widget=RightLine
   properties {
    visibility=SelfHitTestInvisible
    brush {
     bIsDynamicallyLoaded=false
     drawAs=Image
     imageType=NoImage
     mirroring=NoMirror
     resourceName=None
     resourceObject=&M_UI_PaddleGinger_ThinLine
     tiling=NoTile
     imageSize=<104 104>
     margin=<ltrb 0.5 0 0.5 0>
     outlineSettings {
      bUseBrushTransparency=false
      roundingType=HalfHeightRadius
      width=0
      color {colorUseRule=UseColor_Specified, specifiedColor=<rgba 0 0 0 0>}
      cornerRadii=<0 0 0 0>
     }
     tintColor {colorUseRule=UseColor_Specified, specifiedColor=<rgba 1 1 1 1>}
     uVRegion {bIsValid=false, max=<0 0>, min=<0 0>}
    }
    renderTransform {angle=0, scale=<-1 1>, shear=<-2 1>, translation=<0 0>}
   }
   slot {class=HorizontalBoxSlot, properties {horizontalAlignment=HAlign_Right}}
  }
  RightSideInfo {
   widgetClassPath=&VerticalBox
   parent=RightSide
   widget=RightSideInfo
   properties {renderTransform {angle=0, scale=<1 1>, shear=<-4 2>, translation=<-104 4>}}
   slot {class=HorizontalBoxSlot, properties {verticalAlignment=VAlign_Center, padding=<ltrb 8 0 0 0>}}
  }
  PropIndicatorBox {widgetClassPath=&HorizontalBox, parent=RightSideInfo, widget=PropIndicatorBox, slot {class=VerticalBoxSlot}}
  RichText_CyclePropAction {
   widgetClassPath=&CommonRichTextBlock
   bIsVariable=true
   parent=PropIndicatorBox
   widget=RichText_CyclePropAction
   properties {
    bOverrideDefaultStyle=true
    text=Cycle Prop
    textTransformPolicy=ToUpper
    defaultTextStyleOverride {
     overflowPolicy=Clip
     transformPolicy=None
     colorAndOpacity {colorUseRule=UseColor_Specified, specifiedColor=<rgba 0.25 1 1 1>}
     font {
      bForceMonospaced=false
      bMaterialIsStencil=false
      fontMaterial=null
      fontObject=&BurbankSmall
      letterSpacing=0
      monospacedWidth=1
      size=15
      skewAmount=0
      typefaceFontName=Black
      outlineSettings {
       bApplyOutlineToDropShadows=false
       bMiteredCorners=false
       bSeparateFillAlpha=false
       outlineMaterial=&M-FontShadow
       outlineSize=2
       outlineColor=<rgba 0 0.1667 1 0.75>
      }
     }
     highlightColor {colorUseRule=UseColor_Specified, specifiedColor=<rgba 0 0 0 1>}
     highlightShape {
      bIsDynamicallyLoaded=false
      drawAs=Image
      imageType=NoImage
      mirroring=NoMirror
      resourceName=None
      resourceObject=null
      tiling=NoTile
      imageSize=<32 32>
      margin=<ltrb 0 0 0 0>
      outlineSettings {
       bUseBrushTransparency=false
       roundingType=HalfHeightRadius
       width=0
       color {colorUseRule=UseColor_Specified, specifiedColor=<rgba 0 0 0 0>}
       cornerRadii=<0 0 0 0>
      }
      tintColor {colorUseRule=UseColor_Specified, specifiedColor=<rgba 1 1 1 1>}
      uVRegion {bIsValid=false, max=<0 0>, min=<0 0>}
     }
     selectedBackgroundColor {colorUseRule=UseColor_Foreground, specifiedColor=<rgba 1 0 1 1>}
     shadowColorAndOpacity=<rgba 0 0 0 1>
     shadowOffset=<0 0>
     strikeBrush {
      bIsDynamicallyLoaded=false
      drawAs=Image
      imageType=NoImage
      mirroring=NoMirror
      resourceName=None
      resourceObject=null
      tiling=NoTile
      imageSize=<32 32>
      margin=<ltrb 0 0 0 0>
      outlineSettings {
       bUseBrushTransparency=false
       roundingType=HalfHeightRadius
       width=0
       color {colorUseRule=UseColor_Specified, specifiedColor=<rgba 0 0 0 0>}
       cornerRadii=<0 0 0 0>
      }
      tintColor {colorUseRule=UseColor_Specified, specifiedColor=<rgba 1 1 1 1>}
      uVRegion {bIsValid=false, max=<0 0>, min=<0 0>}
     }
     underlineBrush {
      bIsDynamicallyLoaded=false
      drawAs=Image
      imageType=NoImage
      mirroring=NoMirror
      resourceName=None
      resourceObject=null
      tiling=NoTile
      imageSize=<32 32>
      margin=<ltrb 0 0 0 0>
      outlineSettings {
       bUseBrushTransparency=false
       roundingType=HalfHeightRadius
       width=0
       color {colorUseRule=UseColor_Specified, specifiedColor=<rgba 0 0 0 0>}
       cornerRadii=<0 0 0 0>
      }
      tintColor {colorUseRule=UseColor_Specified, specifiedColor=<rgba 1 1 1 1>}
      uVRegion {bIsValid=false, max=<0 0>, min=<0 0>}
     }
    }
    renderTransform {angle=0, scale=<1 1>, shear=<-12 0>, translation=<0 0>}
   }
   slot {class=HorizontalBoxSlot, properties {horizontalAlignment=HAlign_Left, verticalAlignment=VAlign_Center, padding=<ltrb 8 0 0 0>}}
  }
  Index_Text {
   widgetClassPath=&TextBlock
   bIsVariable=true
   parent=PropIndicatorBox
   widget=Index_Text
   properties {
    text="{Index}"
    visibility=SelfHitTestInvisible
    colorAndOpacity {colorUseRule=UseColor_Specified, specifiedColor=<rgba 0.5 1 1 1>}
    font {
     bForceMonospaced=false
     bMaterialIsStencil=false
     fontMaterial=null
     fontObject=&Burbank
     letterSpacing=0
     monospacedWidth=1
     size=13
     skewAmount=0
     typefaceFontName=Burbank-PlayerPower
     outlineSettings {
      bApplyOutlineToDropShadows=true
      bMiteredCorners=false
      bSeparateFillAlpha=true
      outlineMaterial=&M-FontShadow
      outlineSize=2
      outlineColor=<rgba 0 0.1667 1 1>
     }
    }
    renderTransform {angle=0, scale=<1 1>, shear=<-12 0>, translation=<0 0>}
   }
   slot {class=HorizontalBoxSlot, properties {horizontalAlignment=HAlign_Left, verticalAlignment=VAlign_Bottom, padding=<ltrb 4 0 0 0>}}
  }
  PropIndicator {
   widgetClassPath=&HorizontalBox
   parent=RightSideInfo
   widget=PropIndicator
   slot {class=VerticalBoxSlot, properties {padding=<ltrb 8 0 0 0>}}
  }
  PropPips {
   widgetClassPath=&HorizontalBox
   bIsVariable=true
   parent=PropIndicator
   widget=PropPips
   slot {class=HorizontalBoxSlot, properties {verticalAlignment=VAlign_Center}}
  }
  RightArrow {
   widgetClassPath=&Image
   bIsVariable=true
   parent=PropIndicator
   widget=RightArrow
   properties {
    visibility=SelfHitTestInvisible
    brush {
     bIsDynamicallyLoaded=false
     drawAs=Image
     imageType=NoImage
     mirroring=NoMirror
     resourceName=None
     resourceObject=&M_UI_PaddleGinger_Glitch_Triangle
     tiling=NoTile
     imageSize=<30 30>
     margin=<ltrb 0 0 0 0>
     outlineSettings {
      bUseBrushTransparency=false
      roundingType=HalfHeightRadius
      width=0
      color {colorUseRule=UseColor_Specified, specifiedColor=<rgba 0 0 0 0>}
      cornerRadii=<0 0 0 0>
     }
     tintColor {colorUseRule=UseColor_Specified, specifiedColor=<rgba 1 1 1 1>}
     uVRegion {bIsValid=false, max=<0 0>, min=<0 0>}
    }
    renderTransform {angle=0, scale=<-1 1>, shear=<0 0>, translation=<0 0>}
   }
   slot {class=HorizontalBoxSlot, properties {horizontalAlignment=HAlign_Center, verticalAlignment=VAlign_Center}}
  }
  MobileVisibility {
   widgetClassPath=&CommonVisibilityWidgetLegacy
   bIsVariable=true
   parent=VertBox_All
   widget=MobileVisibility
   properties {bShowForTouch=false}
   slot {class=VerticalBoxSlot, properties {horizontalAlignment=HAlign_Center}}
  }
  Activate {
   widgetClassPath=&Overlay
   bIsVariable=true
   parent=MobileVisibility
   widget=Activate
   properties {renderTransform {angle=0, scale=<1 1>, shear=<0 0>, translation=<232 0>}}
   slot {class=BorderSlot}
  }
  Image_113 {
   widgetClassPath=&Image
   bIsVariable=true
   parent=Activate
   widget=Image_113
   properties {
    visibility=SelfHitTestInvisible
    brush {
     bIsDynamicallyLoaded=false
     drawAs=Image
     imageType=NoImage
     mirroring=NoMirror
     resourceName=None
     resourceObject=/MotherGameplay/Items/PaddleGinger/UI/M_UI_PaddleGinger_Bracket
     tiling=NoTile
     imageSize=<128 128>
     margin=<ltrb 0 0 0 0>
     outlineSettings {
      bUseBrushTransparency=false
      roundingType=HalfHeightRadius
      width=0
      color {colorUseRule=UseColor_Specified, specifiedColor=<rgba 0 0 0 0>}
      cornerRadii=<0 0 0 0>
     }
     tintColor {colorUseRule=UseColor_Specified, specifiedColor=<rgba 1 1 1 1>}
     uVRegion {bIsValid=false, max=<0 0>, min=<0 0>}
    }
    renderTransform {angle=0, scale=<1.125 1.125>, shear=<0 0>, translation=<-80 0>}
   }
   slot {class=OverlaySlot, properties {verticalAlignment=VAlign_Fill}}
  }
  Image_269 {
   widgetClassPath=&Image
   bIsVariable=true
   parent=Activate
   widget=Image_269
   properties {
    brush {
     bIsDynamicallyLoaded=false
     drawAs=Image
     imageType=NoImage
     mirroring=NoMirror
     resourceName=None
     resourceObject=/MotherGameplay/Items/PaddleGinger/UI/M_UI_PaddleGinger_EnergyBacking
     tiling=NoTile
     imageSize=<32 80>
     margin=<ltrb 0 0 0 0>
     outlineSettings {
      bUseBrushTransparency=false
      roundingType=HalfHeightRadius
      width=0
      color {colorUseRule=UseColor_Specified, specifiedColor=<rgba 0 0 0 0>}
      cornerRadii=<0 0 0 0>
     }
     tintColor {colorUseRule=UseColor_Specified, specifiedColor=<rgba 1 1 1 1>}
     uVRegion {bIsValid=false, max=<0 0>, min=<0 0>}
    }
    renderTransform {angle=0, scale=<1 1>, shear=<0 0>, translation=<-18 0>}
   }
   slot {class=OverlaySlot, properties {horizontalAlignment=HAlign_Fill, verticalAlignment=VAlign_Center}}
  }
  HorizontalBox_0 {
   widgetClassPath=&HorizontalBox
   parent=Activate
   widget=HorizontalBox_0
   slot {class=OverlaySlot, properties {verticalAlignment=VAlign_Center}}
  }
  Keybind_ActivateAction {
   widgetClassPath=&KeybindWidget
   bInherited=true
   bIsVariable=true
   parent=HorizontalBox_0
   widget=Keybind_ActivateAction
   slot {class=HorizontalBoxSlot, properties {horizontalAlignment=HAlign_Center, verticalAlignment=VAlign_Center}}
  }
  RichText_ActivateAction {
   widgetClassPath=&CommonRichTextBlock
   bIsVariable=true
   parent=HorizontalBox_0
   widget=RichText_ActivateAction
   properties {
    bOverrideDefaultStyle=true
    text=Activate
    textTransformPolicy=ToUpper
    defaultTextStyleOverride {
     overflowPolicy=Clip
     transformPolicy=None
     colorAndOpacity {colorUseRule=UseColor_Specified, specifiedColor=<rgba 1 1 1 1>}
     font {
      bForceMonospaced=false
      bMaterialIsStencil=false
      fontMaterial=null
      fontObject=&BurbankSmall
      letterSpacing=200
      monospacedWidth=1
      size=18
      skewAmount=0
      typefaceFontName=Black
      outlineSettings {
       bApplyOutlineToDropShadows=false
       bMiteredCorners=false
       bSeparateFillAlpha=false
       outlineMaterial=&M-FontShadow
       outlineSize=2
       outlineColor=<rgba 0 0.1667 1 1>
      }
     }
     highlightColor {colorUseRule=UseColor_Specified, specifiedColor=<rgba 0 0 0 1>}
     highlightShape {
      bIsDynamicallyLoaded=false
      drawAs=Image
      imageType=NoImage
      mirroring=NoMirror
      resourceName=None
      resourceObject=null
      tiling=NoTile
      imageSize=<32 32>
      margin=<ltrb 0 0 0 0>
      outlineSettings {
       bUseBrushTransparency=false
       roundingType=HalfHeightRadius
       width=0
       color {colorUseRule=UseColor_Specified, specifiedColor=<rgba 0 0 0 0>}
       cornerRadii=<0 0 0 0>
      }
      tintColor {colorUseRule=UseColor_Specified, specifiedColor=<rgba 1 1 1 1>}
      uVRegion {bIsValid=false, max=<0 0>, min=<0 0>}
     }
     selectedBackgroundColor {colorUseRule=UseColor_Foreground, specifiedColor=<rgba 1 0 1 1>}
     shadowColorAndOpacity=<rgba 0 0 0 1>
     shadowOffset=<0 0>
     strikeBrush {
      bIsDynamicallyLoaded=false
      drawAs=Image
      imageType=NoImage
      mirroring=NoMirror
      resourceName=None
      resourceObject=null
      tiling=NoTile
      imageSize=<32 32>
      margin=<ltrb 0 0 0 0>
      outlineSettings {
       bUseBrushTransparency=false
       roundingType=HalfHeightRadius
       width=0
       color {colorUseRule=UseColor_Specified, specifiedColor=<rgba 0 0 0 0>}
       cornerRadii=<0 0 0 0>
      }
      tintColor {colorUseRule=UseColor_Specified, specifiedColor=<rgba 1 1 1 1>}
      uVRegion {bIsValid=false, max=<0 0>, min=<0 0>}
     }
     underlineBrush {
      bIsDynamicallyLoaded=false
      drawAs=Image
      imageType=NoImage
      mirroring=NoMirror
      resourceName=None
      resourceObject=null
      tiling=NoTile
      imageSize=<32 32>
      margin=<ltrb 0 0 0 0>
      outlineSettings {
       bUseBrushTransparency=false
       roundingType=HalfHeightRadius
       width=0
       color {colorUseRule=UseColor_Specified, specifiedColor=<rgba 0 0 0 0>}
       cornerRadii=<0 0 0 0>
      }
      tintColor {colorUseRule=UseColor_Specified, specifiedColor=<rgba 1 1 1 1>}
      uVRegion {bIsValid=false, max=<0 0>, min=<0 0>}
     }
    }
   }
   slot {class=HorizontalBoxSlot, properties {verticalAlignment=VAlign_Center, padding=<ltrb 8 0 0 0>}}
  }
  TutorialHighlight {widgetClassPath=&TutorialHighlight}
 ]
}
